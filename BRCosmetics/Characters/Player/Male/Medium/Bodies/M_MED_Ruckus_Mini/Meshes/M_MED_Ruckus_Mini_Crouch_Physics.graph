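# Crouch physics asset for the Ruckus Mini medium male player body, built around a secondary "d_" lower-body bone chain.
# Capsules on the d_ pelvis, thighs, calves and feet are linked by constraints, with knees limited to near-hinge motion and wider hips and ankles.
# Unconstrained seat and main-thigh capsules add collision shapes.
# A huge sphere offset below dyn_slope_foot_c forms a nearly flat, ground-like surface near that bone.
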PhysicsAsset /BRCosmetics/Characters/Player/Male/Medium/Bodies/M_MED_Ruckus_Mini/Meshes/M_MED_Ruckus_Mini_Crouch_Physics
bodies {
 d_calf_l [{shapeType=Capsule, length=18.053, radius=2.4493, center=<-10 0 0>, rotation=<rot 90 0 0>}]
 d_calf_r [{shapeType=Capsule, length=18.053, radius=2.4493, center=<10 0 0>, rotation=<rot 90 0 0>}]
 d_foot_l [{shapeType=Capsule, length=4.8877, radius=3.4775, shapeName=d_foot_r_capsule, center=<-4.2377 1.5989 0>, rotation=<rot 90 -26.565 23.4302>}]
 d_foot_r [{
   shapeType=Capsule
   length=5.1377
   radius=3.4775
   shapeName=d_foot_r_capsule
   center=<4.1048 -1.7861 0.0001>
   rotation=<rot 90 -9.4623 40.534>
  }]
 d_pelvis [{shapeType=Capsule, length=16.8117, radius=13.3817, center=<-5.4327 2.6605 0.2658>, rotation=<rot 0.4622 -90 -90.7173>}]
 d_thigh_l [{shapeType=Capsule, length=25, radius=3, center=<-12 0 0>, rotation=<rot 90 0 0>}]
 d_thigh_r [{shapeType=Capsule, length=25, radius=3, center=<12 0 0>, rotation=<rot 90 0 0>}]
 dyn_slope_foot_c [{shapeType=Sphere, radius=200, center=<0 0 -195.9351>}]
 seat [{
   shapeType=Capsule
   length=44.9274
   radius=4.2459
   shapeName=seat_capsule
   center=<-11.0416 17.0189 0.0001>
   rotation=<rot 0 -39.9997 89.9999>
  }]
 thigh_l [{shapeType=Capsule, length=31.755, radius=4.755, shapeName=thigh_l_capsule, center=<-19.154 0.0007 0>, rotation=<rot 90 -0.0198 -0.0199>}]
 thigh_r [{shapeType=Capsule, length=31.755, radius=4.755, shapeName=thigh_l_capsule, center=<19.154 0.0007 0>, rotation=<rot 90 -0.0198 -0.0199>}]
}
constraints [
 {bone1Name=d_calf_l, bone2Name=d_thigh_l, swing1LimitDegrees=45, swing2LimitDegrees=1, twistLimitDegrees=1}
 {bone1Name=d_calf_r, bone2Name=d_thigh_r, swing1LimitDegrees=45, swing2LimitDegrees=1, twistLimitDegrees=1}
 {bone1Name=d_thigh_r, bone2Name=d_pelvis, swing1LimitDegrees=45, swing2LimitDegrees=45, twistLimitDegrees=10}
 {bone1Name=d_thigh_l, bone2Name=d_pelvis, swing1LimitDegrees=45, swing2LimitDegrees=45, twistLimitDegrees=10}
 {bone1Name=d_foot_r, bone2Name=d_calf_r, swing1LimitDegrees=45, swing2LimitDegrees=45, twistLimitDegrees=10}
 {bone1Name=d_foot_l, bone2Name=d_calf_l, swing1LimitDegrees=45, swing2LimitDegrees=45, twistLimitDegrees=10}
]
