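# Control Rig for the M_MED_Ruckus_Mini body, procedurally posing a small seated "d_" skeleton and its chair rig over the main skeleton.
# Torso and small spine aim using the main neck, head_back bends the chair spine, and two-bone IK pins hands to armrests.
# Separate forward-solve chains align hip, femur and patella bones to the main legs.
# Small thighs aim at crouch/dbno-blended leg targets, sphere-clamped around a pelvis offset with debug drawing enabled.
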ControlRigBlueprint /BRCosmetics/Characters/Player/Male/Medium/Bodies/M_MED_Ruckus_Mini/Meshes/M_MED_Ruckus_Mini_CtrlRig
graphs [
 RigVMFunctionLibrary {links [], nodes []}
 RigVMModel {nodes [
   RigUnit_BeginExecution = &RigUnit_BeginExecution()
   n1 = /Script/ControlRig.RigUnit_SequenceExecution(ExecuteContext=@RigUnit_BeginExecution.ExecutePin) {is=struct RigVMUnitNode}
   Item = /Script/ControlRig.RigUnit_Item(Item=(Type=Bone,Name="torso")) {is=struct RigVMUnitNode}
   RigUnit_GetRelativeTransformForItem = &RigUnit_GetRelativeTransformForItem(Child=(Type=Bone,Name="neck_02"), bChildInitial=true, bParentInitial=true, Parent=@Item.Item)
   /Script/ControlRig.RigUnit_AimItem(
    DebugSettings=(Scale=10,WorldOffset=(Rotation=(X=0,Y=0,Z=0,W=1),Translation=(X=0,Y=0,Z=0),Scale3D=(X=1,Y=1,Z=1)))
    Primary=(Weight=1,Axis=(X=1,Y=0,Z=0),Target=(X=0,Y=0,Z=0),Kind=Location,Space=(Type=Bone,Name="neck_02"))
    Secondary=(Weight=0,Axis=(X=0,Y=0,Z=1),Target=(X=1,Y=0,Z=0),Kind=Location,Space=(Type=Bone))
    Weight=0.5
    Item=@Item.Item
    Primary.Axis=@RigUnit_GetRelativeTransformForItem.RelativeTransform.Translation
    ExecutePin=@n1.A
   ) {is=struct RigVMUnitNode}
   n2 = &RigUnit_Item(Item=(Type=Bone,Name="d_spine_01"))
   &RigUnit_AimItem(
    DebugSettings=(Scale=10,WorldOffset=(Rotation=(X=0,Y=0,Z=0,W=1),Translation=(X=0,Y=0,Z=0),Scale3D=(X=1,Y=1,Z=1)))
    Primary=(Weight=1,Axis=(X=1,Y=0,Z=0),Target=(X=0,Y=0,Z=0),Kind=Location,Space=(Type=Bone,Name="neck_02"))
    Secondary=(Weight=0,Axis=(X=0,Y=0,Z=1),Target=(X=1,Y=0,Z=0),Kind=Location,Space=(Type=Bone))
    Weight=1
    Item=@n2.Item
    Primary.Axis=&RigUnit_GetRelativeTransformForItem(Child=(Type=Bone,Name="spine_01_fr_tgt"), bChildInitial=true, bParentInitial=true, Parent=@n2.Item).RelativeTransform.Translation
    ExecutePin=@n1.B
   )
   n3 = &RigVMFunction_MathQuaternionFromEuler(Euler=(X=0,Y=0,Z=-8))
   n4 = RigVMVariableNode() {variable=head_back}
   > &RigUnit_OffsetTransformForItem(
    Item=(Type=Bone,Name="chair_spine_01")
    OffsetTransform=(Rotation=(X=0,Y=0,Z=0,W=1),Translation=(X=0,Y=0,Z=0),Scale3D=(X=1,Y=1,Z=1))
    bPropagateToChildren=True
    OffsetTransform.Rotation=@n3.Result
    Weight=@n4.Value
   )
   > &RigUnit_OffsetTransformForItem(
    Item=(Type=Bone,Name="chair_spine_02")
    OffsetTransform=(Rotation=(X=0,Y=0,Z=0,W=1),Translation=(X=0,Y=0,Z=0),Scale3D=(X=1,Y=1,Z=1))
    bPropagateToChildren=True
    OffsetTransform.Rotation=@n3.Result
    Weight=@n4.Value
   )
   > &RigUnit_OffsetTransformForItem(
    Item=(Type=Bone,Name="chair_spine_03")
    OffsetTransform=(Rotation=(X=0,Y=0,Z=0,W=1),Translation=(X=0,Y=0,Z=0),Scale3D=(X=1,Y=1,Z=1))
    bPropagateToChildren=True
    OffsetTransform.Rotation=@n3.Result
    Weight=@n4.Value
   )
   n5 = /Script/ControlRig.RigUnit_ProjectTransformToNewParent(
    Child=(Type=Bone,Name="d_hand_l")
    NewParent=(Type=Bone,Name="armrest_l")
    OldParent=(Type=Bone,Name="armrest_l")
    bChildInitial=True
    bOldParentInitial=True
   ) {is=struct RigVMUnitNode}
   n6 = &RigUnit_ProjectTransformToNewParent(
    Child=(Type=Bone,Name="d_lowerarm_l")
    NewParent=(Type=Bone,Name="armrest_l")
    OldParent=(Type=Bone,Name="armrest_l")
    bChildInitial=True
    bOldParentInitial=True
   )
   /Script/ControlRig.RigUnit_TwoBoneIKSimplePerItem(
    DebugSettings=(Scale=10,WorldOffset=(Rotation=(X=0,Y=0,Z=0,W=1),Translation=(X=0,Y=0,Z=0),Scale3D=(X=1,Y=1,Z=1)))
    EffectorItem=(Type=Bone,Name="d_hand_l")
    ItemA=(Type=Bone,Name="d_upperarm_l")
    ItemB=(Type=Bone,Name="d_lowerarm_l")
    PoleVectorKind=Location
    PoleVectorSpace=(Type=Bone)
    PrimaryAxis=(X=1,Y=0,Z=0)
    SecondaryAxis=(X=0,Y=1,Z=0)
    SecondaryAxisWeight=0
    StretchMaximumRatio=1.25
    StretchStartRatio=0.75
    Weight=1
    bEnableStretch=false
    bPropagateToChildren=True
    Effector=@n5.Transform
    PoleVector=@n6.Transform.Translation
    ExecutePin=@n1.C
   ) {is=struct RigVMUnitNode}
   n7 = &RigUnit_ProjectTransformToNewParent(
    Child=(Type=Bone,Name="d_hand_r")
    NewParent=(Type=Bone,Name="armrest_r")
    OldParent=(Type=Bone,Name="armrest_r")
    bChildInitial=True
    bOldParentInitial=True
   )
   n8 = &RigUnit_ProjectTransformToNewParent(
    Child=(Type=Bone,Name="d_lowerarm_r")
    NewParent=(Type=Bone,Name="armrest_r")
    OldParent=(Type=Bone,Name="armrest_r")
    bChildInitial=True
    bOldParentInitial=True
   )
   > &RigUnit_TwoBoneIKSimplePerItem(
    DebugSettings=(Scale=10,WorldOffset=(Rotation=(X=0,Y=0,Z=0,W=1),Translation=(X=0,Y=0,Z=0),Scale3D=(X=1,Y=1,Z=1)))
    EffectorItem=(Type=Bone,Name="d_hand_r")
    ItemA=(Type=Bone,Name="d_upperarm_r")
    ItemB=(Type=Bone,Name="d_lowerarm_r")
    PoleVectorKind=Location
    PoleVectorSpace=(Type=Bone)
    PrimaryAxis=(X=-1,Y=0,Z=0)
    SecondaryAxis=(X=0,Y=1,Z=0)
    SecondaryAxisWeight=0
    StretchMaximumRatio=1.25
    StretchStartRatio=0.75
    Weight=1
    bEnableStretch=false
    bPropagateToChildren=True
    Effector=@n7.Transform
    PoleVector=@n8.Transform.Translation
   )
   n9 = &RigUnit_BeginExecution()
   n10 = &RigUnit_ProjectTransformToNewParent(
    Child=(Type=Bone,Name="hip_bot_l")
    NewParent=(Type=Bone,Name="thigh_twist_01_l")
    OldParent=(Type=Bone,Name="thigh_twist_01_l")
    bChildInitial=True
    bOldParentInitial=True
   )
   &RigUnit_SetRotation(Item=(Type=Bone,Name="hip_bot_l"), Space=GlobalSpace, Weight=0.2, Value=@n10.Transform.Rotation, ExecutePin=@n9.ExecutePin)
   > &RigUnit_AimItem(
    DebugSettings=(Scale=10,WorldOffset=(Rotation=(X=0,Y=0,Z=0,W=1),Translation=(X=0,Y=0,Z=0),Scale3D=(X=1,Y=1,Z=1)))
    Item=(Type=Bone,Name="femur_l")
    Primary=(Weight=1,Axis=(X=1,Y=0,Z=0),Target=(X=0,Y=0,Z=0),Kind=Location,Space=(Type=Bone,Name="calf_l"))
    Secondary=(Weight=0,Axis=(X=0,Y=0,Z=1),Target=(X=1,Y=0,Z=0),Kind=Location,Space=(Type=Bone))
    Weight=1
    Primary.Axis=&RigUnit_GetRelativeTransformForItem(Child=(Type=Bone,Name="calf_l"), Parent=(Type=Bone,Name="femur_l"), bChildInitial=true, bParentInitial=true).RelativeTransform.Translation
   )
   > /Script/ControlRig.RigUnit_SetTranslation(
    Item=(Type=Bone,Name="patella_l")
    Space=GlobalSpace
    Weight=1
    bPropagateToChildren=True
    Value=&RigUnit_GetTransform(Item=(Type=Bone,Name="calf_l")).Transform.Translation
   ) {is=struct RigVMUnitNode, template="Set Transform::Execute(in Item,in Space,in bInitial,in Value,in Weight,in bPropagateToChildren)"}
   n11 = &RigUnit_BeginExecution()
   n12 = &RigUnit_ProjectTransformToNewParent(
    Child=(Type=Bone,Name="hip_bot_r")
    NewParent=(Type=Bone,Name="thigh_twist_01_r")
    OldParent=(Type=Bone,Name="thigh_twist_01_r")
    bChildInitial=True
    bOldParentInitial=True
   )
   &RigUnit_SetRotation(Item=(Type=Bone,Name="hip_bot_r"), Space=GlobalSpace, Weight=0.2, Value=@n12.Transform.Rotation, ExecutePin=@n11.ExecutePin)
   > &RigUnit_AimItem(
    DebugSettings=(Scale=10,WorldOffset=(Rotation=(X=0,Y=0,Z=0,W=1),Translation=(X=0,Y=0,Z=0),Scale3D=(X=1,Y=1,Z=1)))
    Item=(Type=Bone,Name="femur_r")
    Primary=(Weight=1,Axis=(X=1,Y=0,Z=0),Target=(X=0,Y=0,Z=0),Kind=Location,Space=(Type=Bone,Name="calf_r"))
    Secondary=(Weight=0,Axis=(X=0,Y=0,Z=1),Target=(X=1,Y=0,Z=0),Kind=Location,Space=(Type=Bone))
    Weight=1
    Primary.Axis=&RigUnit_GetRelativeTransformForItem(Child=(Type=Bone,Name="calf_r"), Parent=(Type=Bone,Name="femur_r"), bChildInitial=true, bParentInitial=true).RelativeTransform.Translation
   )
   > &RigUnit_SetTranslation(
    Item=(Type=Bone,Name="patella_r")
    Space=GlobalSpace
    Weight=1
    bPropagateToChildren=True
    Value=&RigUnit_GetTransform(Item=(Type=Bone,Name="calf_r")).Transform.Translation
   ) {template="Set Transform::Execute(in Item,in Space,in bInitial,in Value,in Weight,in bPropagateToChildren)"}
   n13 = &RigUnit_BeginExecution()
   n14 = /Script/RigVM.RigVMFunction_MathTransformTransformVector(Location=(X=50,Y=-25,Z=0), Transform=&RigUnit_GetTransform(Item=(Type=Bone,Name="pelvis")).Transform) {is=struct RigVMUnitNode}
   AlphaInterp = &RigVMFunction_AlphaInterp(
    InRange=(Max=1)
    InterpSpeedDecreasing=10
    InterpSpeedIncreasing=10
    OutRange=(Max=1)
    Scale=1
    Bias=~
    ClampMin=~
    bMapRange=~
    Value=RigVMVariableNode() {variable=crouch}.Value
   )
   n15 = /Script/RigVM.RigVMFunction_MathFloatLerp(
    B=0.45
    A=&RigVMFunction_MathFloatLerp(A=0.05, B=0.2, T=@AlphaInterp.Result) {template="Interpolate::Execute(in A,in B,in T,out Result)"}.Result
    T=RigVMVariableNode() {variable=dbno}.Value
   ) {is=struct RigVMUnitNode, template="Interpolate::Execute(in A,in B,in T,out Result)"}
   n16 = &RigVMFunction_MathFloatLerp(A=0.4, B=0.15, T=@AlphaInterp.Result) {template="Interpolate::Execute(in A,in B,in T,out Result)"}
   n17 = &RigUnit_ProjectTransformToNewParent(
    Child=(Type=Bone,Name="d_foot_r")
    NewParent=(Type=Bone,Name="thigh_r")
    OldParent=(Type=Bone,Name="thigh_r")
    bChildInitial=True
    bOldParentInitial=True
   )
   n18 = /Script/RigVM.RigVMFunction_MathVectorLerp(
    B=&RigUnit_GetTransform(Item=(Type=Bone,Name="foot_r")).Transform.Translation
    A=&RigVMFunction_MathVectorLerp(A=&RigUnit_GetTransform(Item=(Type=Bone,Name="calf_r")).Transform.Translation, T=@n16.Result, B=@n17.Transform.Translation) {template="Interpolate::Execute(in A,in B,in T,out Result)"}.Result
    T=RigVMVariableNode() {variable=dbno}.Value
   ) {is=struct RigVMUnitNode, template="Interpolate::Execute(in A,in B,in T,out Result)"}
   n19 = &RigUnit_ProjectTransformToNewParent(
    Child=(Type=Bone,Name="d_foot_l")
    NewParent=(Type=Bone,Name="thigh_l")
    OldParent=(Type=Bone,Name="thigh_l")
    bChildInitial=True
    bOldParentInitial=True
   )
   n20 = &RigVMFunction_MathVectorLerp(
    B=&RigUnit_GetTransform(Item=(Type=Bone,Name="foot_l")).Transform.Translation
    A=&RigVMFunction_MathVectorLerp(A=&RigUnit_GetTransform(Item=(Type=Bone,Name="calf_l")).Transform.Translation, T=@n16.Result, B=@n19.Transform.Translation) {template="Interpolate::Execute(in A,in B,in T,out Result)"}.Result
    T=RigVMVariableNode() {variable=dbno}.Value
   ) {template="Interpolate::Execute(in A,in B,in T,out Result)"}
   n21 = /Script/RigVM.RigVMFunction_MathTransformClampSpatially(
    Axis=None
    DebugColor=(R=1,G=0,B=0,A=1)
    DebugThickness=1
    Maximum=1000
    Minimum=50
    Space=(Rotation=(X=0,Y=0,Z=0,W=1),Translation=(X=0,Y=0,Z=0),Scale3D=(X=1,Y=1,Z=1))
    Type=Sphere
    Value=(Rotation=(X=0,Y=0,Z=0,W=1),Translation=(X=0,Y=0,Z=0),Scale3D=(X=1,Y=1,Z=1))
    bDrawDebug=true
    Space.Translation=@n14.Result
    Value.Translation=&RigVMFunction_MathVectorLerp(T=@n15.Result, B=@n18.Result, A=@n20.Result) {template="Interpolate::Execute(in A,in B,in T,out Result)"}.Result
   ) {
    is=struct RigVMUnitNode
    template="ClampSpatially::Execute(in Value,in Axis,in Type,in Minimum,in Maximum,in Space,in bDrawDebug,in DebugColor,in DebugThickness,out Result)"
   }
   &RigUnit_AimItem(
    DebugSettings=(Scale=10,WorldOffset=(Rotation=(X=0,Y=0,Z=0,W=1),Translation=(X=0,Y=0,Z=0),Scale3D=(X=1,Y=1,Z=1)))
    Item=(Type=Bone,Name="d_thigh_l")
    Primary=(Weight=1,Axis=(X=1,Y=0,Z=0),Target=(X=0,Y=0,Z=0),Kind=Location,Space=(Type=Bone))
    Secondary=(Weight=0,Axis=(X=0,Y=0,Z=1),Target=(X=1,Y=0,Z=0),Kind=Location,Space=(Type=Bone))
    Weight=1
    Primary.Axis=&RigUnit_GetRelativeTransformForItem(Child=(Type=Bone,Name="d_foot_l"), Parent=(Type=Bone,Name="d_thigh_l"), bChildInitial=true, bParentInitial=true).RelativeTransform.Translation
    Primary.Target=@n21.Result.Translation
    ExecutePin=@n13.ExecutePin
   )
   n22 = &RigUnit_BeginExecution()
   n23 = &RigVMFunction_MathTransformClampSpatially(
    Axis=None
    DebugColor=(R=1,G=0,B=0,A=1)
    DebugThickness=1
    Maximum=1000
    Minimum=50
    Space=(Rotation=(X=0,Y=0,Z=0,W=1),Translation=(X=0,Y=0,Z=0),Scale3D=(X=1,Y=1,Z=1))
    Type=Sphere
    Value=(Rotation=(X=0,Y=0,Z=0,W=1),Translation=(X=0,Y=0,Z=0),Scale3D=(X=1,Y=1,Z=1))
    bDrawDebug=true
    Space.Translation=@n14.Result
    Value.Translation=&RigVMFunction_MathVectorLerp(T=@n15.Result, A=@n18.Result, B=@n20.Result) {template="Interpolate::Execute(in A,in B,in T,out Result)"}.Result
   ) {template="ClampSpatially::Execute(in Value,in Axis,in Type,in Minimum,in Maximum,in Space,in bDrawDebug,in DebugColor,in DebugThickness,out Result)"}
   &RigUnit_AimItem(
    DebugSettings=(Scale=10,WorldOffset=(Rotation=(X=0,Y=0,Z=0,W=1),Translation=(X=0,Y=0,Z=0),Scale3D=(X=1,Y=1,Z=1)))
    Item=(Type=Bone,Name="d_thigh_r")
    Primary=(Weight=1,Axis=(X=1,Y=0,Z=0),Target=(X=0,Y=0,Z=0),Kind=Location,Space=(Type=Bone))
    Secondary=(Weight=0,Axis=(X=0,Y=0,Z=1),Target=(X=1,Y=0,Z=0),Kind=Location,Space=(Type=Bone))
    Weight=1
    Primary.Axis=&RigUnit_GetRelativeTransformForItem(Child=(Type=Bone,Name="d_foot_r"), Parent=(Type=Bone,Name="d_thigh_r"), bChildInitial=true, bParentInitial=true).RelativeTransform.Translation
    Primary.Target=@n23.Result.Translation
    ExecutePin=@n22.ExecutePin
   )
  ]}
]
hierarchy [
 root {loc=<0 0 0>, rot=<0 0 0 1>}
 pelvis {parent=root, loc=<0 -0.7071 93.1191>, rot=<0.0127 0.707 -0.0127 -0.707>}
 spine_01 {parent=pelvis, loc=<6.7494 -0.6388 0>, rot=<0 0 0.034 -0.9994>}
 spine_02 {parent=spine_01, loc=<7.2517 -0.035 0>, rot=<0 0 0.0238 -0.9997>}
 spine_03 {parent=spine_02, loc=<6.0104 0 0>, rot=<0 0 0.0149 -0.9999>}
 spine_04 {parent=spine_03, loc=<8.1023 0 0>, rot=<0 0 -0.0132 -0.9999>}
 spine_05 {parent=spine_04, loc=<10.0146 0 0>, rot=<0 0 -0.0015 -1>}
 clavicle_l {parent=spine_05, loc=<4.3575 -0.1466 -2.3223>, rot=<0.0465 -0.7289 -0.009 -0.6829>}
 upperarm_l {parent=clavicle_l, loc=<15.5111 0 0>, rot=<0.0226 -0.3234 -0.0088 -0.9459>}
 lowerarm_l {parent=upperarm_l, loc=<26.6513 0 0>, rot=<0.0391 -0.0099 -0.2453 -0.9686>}
 hand_l {parent=lowerarm_l, loc=<22.3491 0 0>, rot=<-0.7081 -0.0507 0.0975 -0.6975>}
 index_metacarpal_l {parent=hand_l, loc=<1.6535 0.0183 -2.96>, rot=<0.0158 -0.0124 -0.0173 -0.9996>}
 index_01_l {parent=index_metacarpal_l, loc=<8.3008 0.0357 -0.0524>, rot=<-0.0044 0.0485 0.0924 -0.9945>}
 index_02_l {parent=index_01_l, loc=<3.3587 -0.3004 0.0452>, rot=<-0.002 0.0027 0.3225 -0.9466>}
 index_03_l {parent=index_02_l, loc=<2.5578 -0.123 -0.0252>, rot=<-0.002 0.0067 0.2558 -0.9667>}
 middle_metacarpal_l {parent=hand_l, loc=<2.1856 -0.1725 -0.5984>, rot=<0.006 0.0168 -0.0329 -0.9993>}
 middle_01_l {parent=middle_metacarpal_l, loc=<7.3275 -0.2447 0.0129>, rot=<-0.0208 0.0042 0.0858 -0.9961>}
 middle_02_l {parent=middle_01_l, loc=<4.0123 0.1401 0.1175>, rot=<-0.0218 0.0258 0.3512 -0.9357>}
 middle_03_l {parent=middle_02_l, loc=<2.6077 0.0534 0.1373>, rot=<0.0082 0.0188 0.1708 -0.9851>}
 pinky_metacarpal_l {parent=hand_l, loc=<1.7616 -0.9753 3.3814>, rot=<-0.0397 0.0814 -0.069 -0.9935>}
 pinky_01_l {parent=pinky_metacarpal_l, loc=<7.1709 -0.4831 -0.1094>, rot=<-0.0654 -0.0035 0.2242 -0.9723>}
 pinky_02_l {parent=pinky_01_l, loc=<2.8274 0.0893 -0.1043>, rot=<0.0037 0.0011 0.2854 -0.9584>}
 pinky_03_l {parent=pinky_02_l, loc=<2.2046 0.0307 -0.0078>, rot=<-0.0002 0.0098 0.1445 -0.9895>}
 ring_metacarpal_l {parent=hand_l, loc=<2.2074 -0.3625 1.3638>, rot=<-0.0284 0.0722 -0.0458 -0.9959>}
 ring_01_l {parent=ring_metacarpal_l, loc=<7.0295 -0.6197 -0.1446>, rot=<-0.0489 -0.0252 0.1892 -0.9804>}
 ring_02_l {parent=ring_01_l, loc=<3.6174 0.4647 -0.1027>, rot=<0 0 0.2806 -0.9598>}
 ring_03_l {parent=ring_02_l, loc=<2.4152 0.114 0.0199>, rot=<0 0 0.1189 -0.9929>}
 thumb_01_l {parent=hand_l, loc=<1.3118 -0.3191 -2.459>, rot=<0.6546 -0.4042 0.0664 -0.6354>}
 thumb_02_l {parent=thumb_01_l, loc=<5.9336 -0.0448 0.0401>, rot=<-0.0037 0.0213 0.298 -0.9543>}
 thumb_03_l {parent=thumb_02_l, loc=<2.2223 0.3682 -0.0907>, rot=<0.0118 -0.013 0.0066 -0.9998>}
 lowerarm_twist_01_l {parent=lowerarm_l, loc=<17.8549 0 0>, rot=<0 0 0 -1>}
 lowerarm_twist_02_l {parent=lowerarm_l, loc=<9.6199 0 0>, rot=<0 0 0 -1>}
 elbow_disc_l {parent=lowerarm_l, loc=<-0.0134 -0.137 -0.0713>, rot=<0.0407 -0.0528 0.9719 0.2256>}
 elbow_bk_l {parent=elbow_disc_l, loc=<-1.5445 6.2793 -0.067>, rot=<0.97 0.2426 -0.0144 0.0053>}
 elbow_fr_l {parent=elbow_disc_l, loc=<4.1587 -4.9036 0.0193>, rot=<0.9814 0.1884 -0.0351 0.0099>}
 upperarm_twist_02_l {parent=upperarm_l, loc=<13.7941 0 0>, rot=<0 0 0 -1>}
 upperarm_twist_01_l {parent=upperarm_l, loc=<0.0041 0 0>, rot=<0 0 0 -1>}
 bicep_l {parent=upperarm_l, loc=<4.0114 -0.8795 0.5171>, rot=<0 -0.0112 -0.019 -0.9998>}
 bicep_end_l {parent=bicep_l, loc=<24.7164 0 0>, rot=<0 0 0 -1>}
 tricep_l {parent=upperarm_l, loc=<4.1305 -4.1536 0.2812>, rot=<0.1467 0 -0.0479 -0.988>}
 tricep_end_l {parent=tricep_l, loc=<24.6791 0 0>, rot=<-0.0153 -0.9879 -0.1544 0.0023>}
 shldr_l {parent=clavicle_l, loc=<18.3404 3.3768 2.058>, rot=<0.0266 -0.9975 -0.0126 -0.0648>}
 clavicle_r {parent=spine_05, loc=<4.3575 -0.1466 2.3223>, rot=<0.7289 0.0465 -0.6829 0.009>}
 upperarm_r {parent=clavicle_r, loc=<-15.5111 0 0>, rot=<0.0226 -0.3234 -0.0088 -0.9459>}
 lowerarm_r {parent=upperarm_r, loc=<-26.6515 0 0>, rot=<0.0391 -0.0099 -0.2453 -0.9686>}
 hand_r {parent=lowerarm_r, loc=<-22.3492 0 0>, rot=<-0.7081 -0.0507 0.0975 -0.6975>}
 index_metacarpal_r {parent=hand_r, loc=<-1.6534 -0.018 2.9601>, rot=<0.0158 -0.0124 -0.0173 -0.9996>}
 index_01_r {parent=index_metacarpal_r, loc=<-8.3008 -0.0358 0.0524>, rot=<-0.0044 0.0485 0.0924 -0.9945>}
 index_02_r {parent=index_01_r, loc=<-3.3589 0.3003 -0.0453>, rot=<-0.002 0.0027 0.3225 -0.9466>}
 index_03_r {parent=index_02_r, loc=<-2.5578 0.1231 0.0251>, rot=<-0.002 0.0067 0.2558 -0.9667>}
 middle_metacarpal_r {parent=hand_r, loc=<-2.1857 0.1727 0.5984>, rot=<0.006 0.0168 -0.0329 -0.9993>}
 middle_01_r {parent=middle_metacarpal_r, loc=<-7.3273 0.2447 -0.0129>, rot=<-0.0208 0.0042 0.0858 -0.9961>}
 middle_02_r {parent=middle_01_r, loc=<-4.0142 -0.1411 -0.07>, rot=<-0.0217 0.0259 0.3512 -0.9357>}
 middle_03_r {parent=middle_02_r, loc=<-2.6074 -0.0534 -0.1312>, rot=<0.0083 0.0188 0.1708 -0.9851>}
 pinky_metacarpal_r {parent=hand_r, loc=<-1.7617 0.9753 -3.3814>, rot=<-0.0397 0.0814 -0.069 -0.9935>}
 pinky_01_r {parent=pinky_metacarpal_r, loc=<-7.1707 0.4832 0.1094>, rot=<-0.0654 -0.0035 0.2242 -0.9723>}
 pinky_02_r {parent=pinky_01_r, loc=<-2.8275 -0.0892 0.1043>, rot=<0.0037 0.0011 0.2854 -0.9584>}
 pinky_03_r {parent=pinky_02_r, loc=<-2.2045 -0.031 0.0078>, rot=<-0.0002 0.0098 0.1445 -0.9895>}
 ring_metacarpal_r {parent=hand_r, loc=<-2.2073 0.3627 -1.3638>, rot=<-0.0284 0.0722 -0.0458 -0.9959>}
 ring_01_r {parent=ring_metacarpal_r, loc=<-7.0297 0.6197 0.1446>, rot=<-0.0489 -0.0252 0.1892 -0.9804>}
 ring_02_r {parent=ring_01_r, loc=<-3.6172 -0.4647 0.1027>, rot=<0 0 0.2806 -0.9598>}
 ring_03_r {parent=ring_02_r, loc=<-2.4152 -0.1141 -0.0199>, rot=<0 0 0.1189 -0.9929>}
 thumb_01_r {parent=hand_r, loc=<-1.3119 0.3193 2.459>, rot=<0.6546 -0.4042 0.0664 -0.6354>}
 thumb_02_r {parent=thumb_01_r, loc=<-5.9336 0.0447 -0.0401>, rot=<-0.0037 0.0213 0.298 -0.9543>}
 thumb_03_r {parent=thumb_02_r, loc=<-2.2223 -0.3682 0.0907>, rot=<0.0118 -0.013 0.0066 -0.9998>}
 lowerarm_twist_01_r {parent=lowerarm_r, loc=<-17.855 0 0>, rot=<0 0 0 -1>}
 lowerarm_twist_02_r {parent=lowerarm_r, loc=<-9.6199 0 0>, rot=<0 0 0 -1>}
 elbow_disc_r {parent=lowerarm_r, loc=<0.0134 0.137 0.0714>, rot=<-0.0528 -0.0407 0.2256 -0.9719>}
 elbow_fr_r {parent=elbow_disc_r, loc=<4.1587 -4.9036 -0.0193>, rot=<0.9814 0.1884 0.0351 -0.0099>}
 elbow_bk_r {parent=elbow_disc_r, loc=<-1.5445 6.2793 0.067>, rot=<0.97 0.2426 0.0144 -0.0053>}
 upperarm_twist_02_r {parent=upperarm_r, loc=<-13.7941 0 0>, rot=<0 0 0 -1>}
 upperarm_twist_01_r {parent=upperarm_r, loc=<-0.0041 0 0>, rot=<0 0 0 -1>}
 tricep_r {parent=upperarm_r, loc=<-4.1306 4.1536 -0.2812>, rot=<0 0.1467 0.988 -0.0479>}
 tricep_end_r {parent=tricep_r, loc=<24.6792 0 0>, rot=<-0.0153 -0.9879 0.1544 -0.0023>}
 bicep_r {parent=upperarm_r, loc=<-4.0115 0.8795 -0.517>, rot=<0.0112 0 0.9998 -0.019>}
 bicep_end_r {parent=bicep_r, loc=<24.7164 0 0>, rot=<0 0 0 -1>}
 shldr_r {parent=clavicle_r, loc=<-18.3404 -3.3768 -2.0579>, rot=<0.9975 0.0266 0.0648 -0.0126>}
 neck_01 {parent=spine_05, loc=<11.3138 0.1529 0>, rot=<0 0 -0.2003 -0.9797>}
 neck_02 {parent=neck_01, loc=<4.5627 0 0>, rot=<0 0 0.0504 -0.9987>}
 head {parent=neck_02, loc=<5.3914 0 0>, rot=<0 0 0.115 -0.9934>}
 d_chest_end_tgt {parent=head, loc=<-13.3974 2.5756 0>, rot=<0 0 0.036 -0.9994>}
 spine_01_fr_tgt {parent=spine_05, loc=<-0.3101 6.4603 0>, rot=<0 0 -0.024 -0.9997>}
 spine_01_bk_tgt {parent=spine_05, loc=<0.3025 -5.7245 0>, rot=<0 0 -0.024 -0.9997>}
 thigh_l {parent=pelvis, loc=<-2.5693 -0.0109 -7.7501>, rot=<0.0012 0.0572 0.0155 -0.9982>}
 calf_l {parent=thigh_l, loc=<-39.0484 0 0>, rot=<0.0006 -0.0161 -0.0391 -0.9991>}
 calf_twist_01_l {parent=calf_l, loc=<-35.813 -0.0967 0.0397>, rot=<0.0237 0.0095 -0.0035 -0.9997>}
 calf_twist_02_l {parent=calf_l, loc=<-18.8064 -0.0967 0.0397>, rot=<0.0237 0.0095 -0.0035 -0.9997>}
 foot_l {parent=calf_l, loc=<-42.561 0.1077 0.4292>, rot=<0.0418 -0.0455 0.0343 -0.9975>}
 ball_l {parent=foot_l, loc=<-6.1599 13.8264 0.6806>, rot=<0.0001 -0.0001 -0.7063 -0.7079>}
 tibia_l_start {parent=calf_l, loc=<1.164 1.5033 0>, rot=<0 0 0 1>}
 tibia_l_end {parent=calf_l, loc=<-2.956 2.5738 0>, rot=<0 0 -0.1053 -0.9944>}
 thigh_twist_01_l {parent=thigh_l, loc=<-21.6303 0 0>, rot=<0.0013 0 0 -1>}
 thigh_r {parent=pelvis, loc=<-2.5691 -0.0109 7.7501>, rot=<0.0572 -0.0012 0.9982 0.0155>}
 calf_r {parent=thigh_r, loc=<39.0487 0 0>, rot=<0.0006 -0.0161 -0.0391 -0.9991>}
 calf_twist_01_r {parent=calf_r, loc=<35.8132 0.0967 -0.0397>, rot=<0.0237 0.0095 -0.0035 -0.9997>}
 calf_twist_02_r {parent=calf_r, loc=<18.8064 0.0967 -0.0397>, rot=<0.0237 0.0095 -0.0035 -0.9997>}
 foot_r {parent=calf_r, loc=<42.5612 -0.1077 -0.4292>, rot=<0.0418 -0.0455 0.0343 -0.9975>}
 ball_r {parent=foot_r, loc=<6.1598 -13.8264 -0.6806>, rot=<0.0001 -0.0001 -0.7063 -0.7079>}
 tibia_r_start {parent=calf_r, loc=<-1.164 -1.5033 0>, rot=<0 0 1 0>}
 tibia_r_end {parent=calf_r, loc=<2.956 -2.5738 0>, rot=<0 0 0.9944 -0.1053>}
 thigh_twist_01_r {parent=thigh_r, loc=<21.6303 0 0>, rot=<0.0013 0 0 -1>}
 torso {parent=pelvis, loc=<4.6994 -2.6291 0>, rot=<0 0 0 -1>}
 d_pelvis {parent=torso, loc=<4.0514 7.0022 0>, rot=<0 0 0 -1>}
 d_spine_01 {parent=d_pelvis, loc=<4.7245 -0.4471 0>, rot=<0 0 0.034 -0.9994>}
 d_spine_02 {parent=d_spine_01, loc=<5.0763 -0.0245 0>, rot=<0 0 0.0238 -0.9997>}
 d_spine_03 {parent=d_spine_02, loc=<4.2073 0 0>, rot=<0 0 0.0149 -0.9999>}
 d_spine_04 {parent=d_spine_03, loc=<5.6716 0 0>, rot=<0 0 -0.0132 -0.9999>}
 d_spine_05 {parent=d_spine_04, loc=<7.0102 0 0>, rot=<0 0 -0.0015 -1>}
 d_clavicle_l {parent=d_spine_05, loc=<4.3575 -0.1466 -2.3223>, rot=<0.1638 -0.8404 0.1109 -0.5047>}
 d_upperarm_l {parent=d_clavicle_l, loc=<5.3286 0 0>, rot=<0.1249 -0.2687 -0.0858 -0.9512>}
 d_lowerarm_l {parent=d_upperarm_l, loc=<19.3549 0 0>, rot=<0 0 -0.778 -0.6282>}
 d_hand_l {parent=d_lowerarm_l, loc=<14.6708 0 0>, rot=<-0.2675 0.0075 0.0656 -0.9613>}
 d_index_metacarpal_l {parent=d_hand_l, loc=<1.2363 -0.1818 -1.4079>, rot=<0.0536 -0.0679 0 -0.9963>}
 d_index_01_l {parent=d_index_metacarpal_l, loc=<3.5779 0.0037 -0.0658>, rot=<-0.0409 0.0465 0.2898 -0.9551>}
 d_index_02_l {parent=d_index_01_l, loc=<1.3505 0 0>, rot=<0 0 0.1985 -0.9801>}
 d_index_03_l {parent=d_index_02_l, loc=<1.0717 0 0>, rot=<0 0 0.138 -0.9904>}
 d_middle_metacarpal_l {parent=d_hand_l, loc=<1.3297 -0.2275 -0.4408>, rot=<0.0384 -0.0239 0.0175 -0.9988>}
 d_middle_01_l {parent=d_middle_metacarpal_l, loc=<3.7875 0 0.0082>, rot=<0 0 0.2771 -0.9608>}
 d_middle_02_l {parent=d_middle_01_l, loc=<1.334 0 0>, rot=<0 0 0.2029 -0.9792>}
 d_middle_03_l {parent=d_middle_02_l, loc=<1.0257 0 0>, rot=<0 0 0.1097 -0.994>}
 d_pinky_metacarpal_l {parent=d_hand_l, loc=<0.8893 -0.0881 1.2253>, rot=<0.032 0.1118 0.0219 -0.993>}
 d_pinky_01_l {parent=d_pinky_metacarpal_l, loc=<3.4251 -0.3012 -0.0016>, rot=<-0.0208 -0.0923 0.1849 -0.9782>}
 d_pinky_02_l {parent=d_pinky_01_l, loc=<1.3841 0 0>, rot=<0 0 0.2697 -0.9629>}
 d_pinky_03_l {parent=d_pinky_02_l, loc=<0.7976 0 0>, rot=<0 0 0.175 -0.9846>}
 d_ring_metacarpal_l {parent=d_hand_l, loc=<1.1792 -0.3098 0.5485>, rot=<-0.0104 0.0302 -0.0228 -0.9992>}
 d_ring_01_l {parent=d_ring_metacarpal_l, loc=<3.5883 0 0>, rot=<0 0 0.3516 -0.9361>}
 d_ring_02_l {parent=d_ring_01_l, loc=<1.6221 0 0>, rot=<0 0 0.1843 -0.9829>}
 d_ring_03_l {parent=d_ring_02_l, loc=<0.9004 0 0>, rot=<0 0 0.1126 -0.9936>}
 d_thumb_01_l {parent=d_hand_l, loc=<1.2709 -0.2531 -1.5761>, rot=<0.6269 -0.3972 -0.217 -0.6342>}
 d_thumb_02_l {parent=d_thumb_01_l, loc=<1.7894 0 0>, rot=<0 0 0.1597 -0.9872>}
 d_thumb_03_l {parent=d_thumb_02_l, loc=<1.3297 0 0>, rot=<0 0 0.0397 -0.9992>}
 d_lowerarm_twist_01_l {parent=d_lowerarm_l, loc=<10 0 0>, rot=<0 0 0 -1>}
 d_lowerarm_twist_02_l {parent=d_lowerarm_l, loc=<4 0 0>, rot=<0 0 0 -1>}
 d_upperarm_twist_01_l {parent=d_upperarm_l, loc=<0 0 0>, rot=<0 0 0 -1>}
 d_upperarm_twist_02_l {parent=d_upperarm_l, loc=<10 0 0>, rot=<0 0 0 -1>}
 d_clavicle_r {parent=d_spine_05, loc=<4.3575 -0.1466 2.3223>, rot=<0.8404 0.1638 -0.5047 -0.1109>}
 d_upperarm_r {parent=d_clavicle_r, loc=<-5.3286 0 0>, rot=<0.1249 -0.2687 -0.0858 -0.9512>}
 d_lowerarm_r {parent=d_upperarm_r, loc=<-19.355 0 0>, rot=<0 0 -0.778 -0.6282>}
 d_hand_r {parent=d_lowerarm_r, loc=<-14.6708 0 0>, rot=<-0.2675 0.0075 0.0656 -0.9613>}
 d_index_metacarpal_r {parent=d_hand_r, loc=<-1.2363 0.1818 1.4079>, rot=<0.0536 -0.0679 0 -0.9963>}
 d_index_01_r {parent=d_index_metacarpal_r, loc=<-3.5779 -0.0037 0.0658>, rot=<-0.0409 0.0465 0.2898 -0.9551>}
 d_index_02_r {parent=d_index_01_r, loc=<-1.3505 0 0>, rot=<0 0 0.1985 -0.9801>}
 d_index_03_r {parent=d_index_02_r, loc=<-1.0717 0 0>, rot=<0 0 0.138 -0.9904>}
 d_middle_metacarpal_r {parent=d_hand_r, loc=<-1.3297 0.2274 0.4408>, rot=<0.0384 -0.0239 0.0175 -0.9988>}
 d_middle_01_r {parent=d_middle_metacarpal_r, loc=<-3.7875 0 -0.0082>, rot=<0 0 0.2771 -0.9608>}
 d_middle_02_r {parent=d_middle_01_r, loc=<-1.334 0 0>, rot=<0 0 0.2029 -0.9792>}
 d_middle_03_r {parent=d_middle_02_r, loc=<-1.0258 0 0>, rot=<0 0 0.1097 -0.994>}
 d_pinky_metacarpal_r {parent=d_hand_r, loc=<-0.8893 0.0881 -1.2253>, rot=<0.032 0.1118 0.0219 -0.993>}
 d_pinky_01_r {parent=d_pinky_metacarpal_r, loc=<-3.4251 0.3012 0.0016>, rot=<-0.0208 -0.0923 0.1849 -0.9782>}
 d_pinky_02_r {parent=d_pinky_01_r, loc=<-1.384 0 0>, rot=<0 0 0.2697 -0.9629>}
 d_pinky_03_r {parent=d_pinky_02_r, loc=<-0.7976 0 0>, rot=<0 0 0.175 -0.9846>}
 d_ring_metacarpal_r {parent=d_hand_r, loc=<-1.1792 0.3098 -0.5485>, rot=<-0.0104 0.0302 -0.0228 -0.9992>}
 d_ring_01_r {parent=d_ring_metacarpal_r, loc=<-3.5883 0 0>, rot=<0 0 0.3516 -0.9361>}
 d_ring_02_r {parent=d_ring_01_r, loc=<-1.6221 0 0>, rot=<0 0 0.1843 -0.9829>}
 d_ring_03_r {parent=d_ring_02_r, loc=<-0.9003 0 0>, rot=<0 0 0.1126 -0.9936>}
 d_thumb_01_r {parent=d_hand_r, loc=<-1.2709 0.2531 1.576>, rot=<0.6269 -0.3972 -0.217 -0.6342>}
 d_thumb_02_r {parent=d_thumb_01_r, loc=<-1.7894 0 0>, rot=<0 0 0.1597 -0.9872>}
 d_thumb_03_r {parent=d_thumb_02_r, loc=<-1.3297 0 0>, rot=<0 0 0.0397 -0.9992>}
 d_lowerarm_twist_01_r {parent=d_lowerarm_r, loc=<-10 0 0>, rot=<0 0 0 -1>}
 d_lowerarm_twist_02_r {parent=d_lowerarm_r, loc=<-4 0 0>, rot=<0 0 0 -1>}
 d_upperarm_twist_01_r {parent=d_upperarm_r, loc=<0 0 0>, rot=<0 0 0 1>}
 d_upperarm_twist_02_r {parent=d_upperarm_r, loc=<-10 0 0>, rot=<0 0 0 -1>}
 d_neck_01 {parent=d_spine_05, loc=<6.6177 -0.4485 0>, rot=<0 0 0 1>}
 d_neck_02 {parent=d_neck_01, loc=<1.7842 0.1582 0>, rot=<0 0 0 1>}
 d_head {parent=d_neck_02, loc=<1.966 0.1466 0>, rot=<0 0 -0.036 -0.9994>}
 d_chest {parent=d_spine_05, loc=<5.1839 -0.0719 0>, rot=<0 0 0 1>}
 d_chest_end {parent=d_spine_05, loc=<10.3678 -0.1438 0>, rot=<0 0 -0.036 -0.9994>}
 d_thigh_l {parent=d_pelvis, loc=<-2.6963 3.0433 -5.8222>, rot=<0.1507 0.0697 0.2436 -0.9555>}
 d_calf_l {parent=d_thigh_l, loc=<-22.3079 0 0>, rot=<0 0 -0.2392 -0.971>}
 d_calf_twist_01_l {parent=d_calf_l, loc=<-13.1174 0.7704 -0.0311>, rot=<0 -0.0056 -0.0013 -1>}
 d_calf_twist_02_l {parent=d_calf_l, loc=<-6.3401 0.6927 0.2234>, rot=<0 -0.0056 -0.0013 -1>}
 d_foot_l {parent=d_calf_l, loc=<-21.0095 0 0>, rot=<-0.1178 -0.0367 0.0026 -0.9924>}
 d_ball_l {parent=d_foot_l, loc=<-4.2881 0.4929 -0.2869>, rot=<0 -0.0001 0 -1>}
 d_thigh_twist_01_l {parent=d_thigh_l, loc=<-10 0 0>, rot=<0 0 0 -1>}
 d_thigh_r {parent=d_pelvis, loc=<-2.6962 3.0433 5.8222>, rot=<0.0697 -0.1507 0.9555 0.2436>}
 d_calf_r {parent=d_thigh_r, loc=<22.3079 0 0>, rot=<0 0 -0.2392 -0.971>}
 d_calf_twist_01_r {parent=d_calf_r, loc=<13.1174 -0.7704 0.0311>, rot=<0 -0.0056 -0.0013 -1>}
 d_calf_twist_02_r {parent=d_calf_r, loc=<6.3401 -0.6927 -0.2234>, rot=<0 -0.0056 -0.0013 -1>}
 d_foot_r {parent=d_calf_r, loc=<21.0095 0 0>, rot=<-0.1191 -0.0367 0.0026 -0.9922>}
 d_ball_r {parent=d_foot_r, loc=<4.2881 -0.4929 0.2869>, rot=<0 -0.0001 0 -1>}
 d_thigh_twist_01_r {parent=d_thigh_r, loc=<10 0 0>, rot=<0 0 0 1>}
 seat {parent=torso, loc=<0 0 0>, rot=<0 0 0 -1>}
 hip_bot_r {parent=seat, loc=<-9.4078 2.4138 9.4667>, rot=<0.0373 -0.0007 0.9991 0.0191>}
 femur_r {parent=hip_bot_r, loc=<6.2689 0 0>, rot=<0 0 0 -1>}
 patella_r {parent=femur_r, loc=<30.4958 0 0>, rot=<0 0 0 -1>}
 thigh_offset_r {parent=femur_r, loc=<0 0 0>, rot=<0 0 0 -1>}
 hip_offset_r {parent=hip_bot_r, loc=<6.2689 0 0>, rot=<0 0 0 -1>}
 hip_bot_l {parent=seat, loc=<-9.4078 2.4138 -9.4667>, rot=<-0.0373 0.0007 0.9991 0.0191>}
 femur_l {parent=hip_bot_l, loc=<6.2689 0 0>, rot=<0 0 0 1>}
 patella_l {parent=femur_l, loc=<30.4958 0 0>, rot=<0 0 0 -1>}
 thigh_offset_l {parent=femur_l, loc=<0 0 0>, rot=<0 0 0 1>}
 hip_offset_l {parent=hip_bot_l, loc=<6.2689 0 0>, rot=<0 0 0 1>}
 armrest_r {parent=seat, loc=<12.382 -6.7882 16.5716>, rot=<0 0 -0.6943 -0.7197>}
 armrest_l {parent=seat, loc=<12.3819 -6.7882 -16.5716>, rot=<0 0 -0.6943 -0.7197>}
 chair_spine_01 {parent=seat, loc=<1.4921 -6.1911 0>, rot=<0 0 0.0179 -0.9998>}
 chair_spine_02 {parent=chair_spine_01, loc=<7.2517 0 0>, rot=<0 0 0 -1>}
 chair_spine_03 {parent=chair_spine_02, loc=<6.0104 0 0>, rot=<0 0 0 -1>}
 chair_spine_04 {parent=chair_spine_03, loc=<8.1023 0 0>, rot=<0 0 0 -1>}
 chair_spine_05 {parent=chair_spine_04, loc=<10.0146 0 0>, rot=<0 0 0 -1>}
 chair_clavicle_l {parent=chair_spine_05, loc=<4.4889 0 -2.3223>, rot=<0.0172 -0.7302 0.0184 -0.6828>}
 link_start_l {parent=chair_clavicle_l, loc=<-0.0883 2.5996 -0.5418>, rot=<0.0168 0.0067 -0.0394 -0.9991>}
 link_end_l {parent=link_start_l, loc=<17.6103 0 0>, rot=<0 0 0 -1>}
 chair_clavicle_r {parent=chair_spine_05, loc=<4.489 0 2.3223>, rot=<0.7302 0.0172 -0.6828 -0.0184>}
 link_start_r {parent=chair_clavicle_r, loc=<0.0883 -2.5996 0.5419>, rot=<-0.0067 0.0168 0.9991 -0.0394>}
 link_end_r {parent=link_start_r, loc=<17.6103 0 0>, rot=<0 0 0 1>}
 dyn_slope_foot_c {parent=root, loc=<0 0 0>, rot=<0 0 1 0>}
]
variables [
 crouch {type=double, default=0}
 dbno {type=double, default=0}
 head_back {type=double, default=0}
]
