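# Post-copy AnimBlueprint for the Ruckus Mini player skin, animating its secondary "d_" body, chair and shoulder pads atop the parent mesh's pose.
# Control Rigs, RigidBody leg physics, idles and bone offsets layer through cached poses into an arm-beam rig.
# A front-end idle curve swaps physics and bypasses shoulder-pad rigging.
# Event Graph derives blends from stance, weapon, montages, slope and bone twist, disabling standing leg physics during listed vehicle emotes.
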AnimBlueprint /BRCosmetics/Characters/Player/Male/Medium/Bodies/M_MED_Ruckus_Mini/Meshes/M_MED_Ruckus_Mini_AnimBP
target_skeleton=&/M_MED_Ruckus_Mini_Skeleton
graphs [
 AnimGraph {nodes [
   n1 = Get dbno_alpha() -> (dbno_alpha:double)
   n2 = "Transform (Modify) Bone - Bone: d_thigh_r"(Alpha=@n1, Translation="0,0,0", Rotation="0,-20,0", Scale="1,1,1") {
    rotationMode=BMM_Additive
    rotationSpace=BCS_ParentBoneSpace
    translationSpace=BCS_ParentBoneSpace
    alphaBoolBlend {blendOption=Linear}
    alphaScaleBiasClamp {bInterpResult=true, interpSpeedIncreasing=3}
    boneToModify {boneName=d_thigh_r}
    rotation {yaw=-20}
   }
   "Transform (Modify) Bone - Bone: d_thigh_l"(ComponentPose=@n2, Alpha=@n1, Translation="0,0,0", Rotation="0,-20,0", Scale="1,1,1") {
    rotationMode=BMM_Additive
    rotationSpace=BCS_ParentBoneSpace
    translationSpace=BCS_ParentBoneSpace
    alphaBoolBlend {blendOption=Linear}
    alphaScaleBiasClamp {bInterpResult=true, interpSpeedIncreasing=3}
    boneToModify {boneName=d_thigh_l}
    rotation {yaw=-20}
   }
   "Transform (Modify) Bone - Bone: spine_01"(Alpha=Get dumbbell() -> (dumbbell:double), Translation="0,0,0", Rotation="0,-10,0", Scale="1,1,1") {
    rotationMode=BMM_Additive
    rotationSpace=BCS_ParentBoneSpace
    alphaBoolBlend {blendOption=Linear}
    alphaScaleBiasClamp {bInterpResult=true, interpSpeedIncreasing=5}
    boneToModify {boneName=spine_01}
    rotation {yaw=-10}
   }
   n3 = "Transform (Modify) Bone - Bone: d_thigh_r"(Alpha=@n1, Translation="0,0,0", Rotation="0,-110,0", Scale="1,1,1") {
    rotationMode=BMM_Additive
    rotationSpace=BCS_ParentBoneSpace
    translationSpace=BCS_ParentBoneSpace
    alphaBoolBlend {blendOption=Linear}
    alphaScaleBiasClamp {bInterpResult=true, scale=0}
    boneToModify {boneName=d_thigh_r}
    rotation {yaw=-110}
   }
   "Transform (Modify) Bone - Bone: d_thigh_l"(ComponentPose=@n3, Alpha=@n1, Translation="0,0,0", Rotation="0,-110,0", Scale="1,1,1") {
    rotationMode=BMM_Additive
    rotationSpace=BCS_ParentBoneSpace
    translationSpace=BCS_ParentBoneSpace
    alphaBoolBlend {blendOption=Linear}
    alphaScaleBiasClamp {bInterpResult=true, scale=0}
    boneToModify {boneName=d_thigh_l}
    rotation {yaw=-110}
   }
   n4 = "Transform (Modify) Bone - Bone: chair_spine_03"!(ComponentPose, Alpha=1, Translation="0,-2,0", Rotation="0,0,-10") -> (Pose) {
    rotationMode=BMM_Additive
    translationMode=BMM_Additive
    alphaBoolBlend {blendInTime=0.3, blendOption=Linear, blendOutTime=0.5}
    boneToModify {boneName=chair_spine_03}
    rotation {roll=-10}
    translation {y=-2}
   }
   n5 = "Transform (Modify) Bone - Bone: chair_spine_04"!(ComponentPose=@n4, Alpha=1, Translation="0,-3,0", Rotation="0.000,0.000,-7") -> (Pose) {
    rotationMode=BMM_Additive
    translationMode=BMM_Additive
    alphaBoolBlend {blendInTime=0.3, blendOption=Linear, blendOutTime=0.5}
    boneToModify {boneName=chair_spine_04}
    rotation {roll=-7}
    translation {y=-3}
   }
   "Transform (Modify) Bone - Bone: chair_spine_05"!(ComponentPose=@n5, Alpha=1, Translation="0,-3,0", Rotation="0.000,0.000,-7") -> (Pose) {
    rotationMode=BMM_Additive
    translationMode=BMM_Additive
    alphaBoolBlend {blendInTime=0.3, blendOption=Linear, blendOutTime=0.5}
    boneToModify {boneName=chair_spine_05}
    rotation {roll=-7}
    translation {y=-3}
   }
   n6 = Use cached pose 'Pre ShldrPad Rigging'()
   n7 = Use cached pose 'Shldr Pad rigging applied'()
   n8 = Get PlayingFrontendIdle() -> (PlayingFrontendIdle:bool)
   n9 = "Transform (Modify) Bone - Bone: link_start_r"(
    ComponentPose=Local To Component(LocalPose=Blend Poses by bool(BlendPose_0=@n6, BlendPose_1=@n7, bActiveValue=@n8) {blendType=Linear})
    Alpha=1
    Translation="12,0,0"
    Rotation="0,0,0"
    Scale="1,1,1"
   ) {
    translationMode=BMM_Additive
    translationSpace=BCS_BoneSpace
    alphaBoolBlend {blendOption=Linear}
    boneToModify {boneName=link_start_r}
    translation {x=12}
   }
   n10 = "Transform (Modify) Bone - Bone: link_start_l"(ComponentPose=@n9, Alpha=1, Translation="12,0,0", Rotation="0,0,0", Scale="1,1,1") {
    translationMode=BMM_Additive
    translationSpace=BCS_BoneSpace
    alphaBoolBlend {blendOption=Linear}
    boneToModify {boneName=link_start_l}
    translation {x=12}
   }
   n11 = Control Rig(Source=Component To Local(ComponentPose=@n10)) {
    alphaBoolBlend {blendOption=Linear}
    controlRigAssetReference {blueprintRigClass=&/M_MED_Ruckus_Mini_Arm_Beams_CtrlRig.M_MED_Ruckus_Mini_Arm_Beams_CtrlRig_C}
   }
   n12 = "Transform (Modify) Bone - Bone: shldr_r"(ComponentPose=Local To Component(LocalPose=@n11), Alpha=Get ADS() -> (ADS:double), Translation="0,0,-2", Rotation="0,0,0", Scale="1,1,1") {
    rotationMode=BMM_Additive
    translationMode=BMM_Additive
    alphaBoolBlend {blendOption=Linear}
    boneToModify {boneName=shldr_r}
    translation {z=-2}
   }
   n13 = "Transform (Modify) Bone - Bone: shldr_r"(ComponentPose=@n12, Alpha=Get crouchADS() -> (crouchADS:double), Translation="0,0,-2", Rotation="0,0,0", Scale="1,1,1") {
    rotationMode=BMM_Additive
    translationMode=BMM_Additive
    alphaBoolBlend {blendOption=Linear}
    boneToModify {boneName=shldr_r}
    translation {z=-2}
   }
   Output Pose(Result=Component To Local(ComponentPose=@n13)) {layerGroup=None}
   n14 = Layered blend per bone(
    BasePose=Use cached pose 'Copied Pose'()
    BlendPoses_0="Sequence Evaluator (by time) 'Ruckus_Mini_Male_Body_Idle'"!(ExplicitTime:float=Get FrontEndIdlePosition() -> (FrontEndIdlePosition:double)) -> (Pose:Pose Link Structure) {
     is=title AnimGraphNode_SequenceEvaluator
     sequence=/BRCosmetics/Animation/Game/MainPlayer/Menu/BR/CustomSkins/_S17/Ruckus_Mini/Ruckus_Mini_Male_Body_Idle
    }
    BlendWeights_0=Get FrontEndIdle() -> (FrontEndIdle:double)
   ) {layerSetup [{branchFilters [{boneName=d_pelvis}, {boneName=shldr_r}, {boneName=shldr_l}]}]}
   n15 = Blend Poses by bool(BlendPose_0=@n14, BlendPose_1=Use cached pose 'Copied Pose'(), bActiveValue=Get PlayingFrontendIdle()) {blendType=Linear}
   n16 = Control Rig!(
    Source=Component To Local(ComponentPose=Local To Component(LocalPose=@n15))
    Alpha
    head_back:double=Get head_back() -> (head_back:double)
    crouch:double=Get crouch_or_DBNO_alpha() -> (crouch_or_DBNO_alpha:double)
    dbno:double=Get dbno_alpha()
   ) -> (Pose) {
    alphaBoolBlend {blendOption=Linear}
    controlRigAssetReference {blueprintRigClass=&/M_MED_Ruckus_Mini_CtrlRig.M_MED_Ruckus_Mini_CtrlRig_C}
   }
   n17 = Get dumbbell()
   n18 = "Transform (Modify) Bone - Bone: d_spine_02"(ComponentPose=Local To Component(LocalPose=@n16), Alpha=@n17, Translation="0,0,0", Rotation="0,-10,0", Scale="1,1,1") {
    rotationMode=BMM_Additive
    rotationSpace=BCS_ParentBoneSpace
    alphaBoolBlend {blendOption=Linear}
    alphaScaleBiasClamp {bInterpResult=true, interpSpeedIncreasing=5}
    boneToModify {boneName=d_spine_02}
    rotation {yaw=-10}
   }
   n19 = "Transform (Modify) Bone - Bone: chair_spine_01"(ComponentPose=@n18, Alpha=@n17, Translation="0,0,0", Rotation="0,-20,0", Scale="1,1,1") {
    rotationMode=BMM_Additive
    rotationSpace=BCS_ParentBoneSpace
    alphaBoolBlend {blendOption=Linear}
    alphaScaleBiasClamp {bInterpResult=true, interpSpeedIncreasing=5}
    boneToModify {boneName=chair_spine_01}
    rotation {yaw=-20}
   }
   n20 = "Transform (Modify) Bone - Bone: dyn_slope_foot_c"!(
    ComponentPose=@n19
    bAlphaBoolEnabled:bool=Get Not Custom Anim() -> (Not Custom Anim:bool)
    Translation=Get slope_foot_trans() -> (slope_foot_trans:Vector)
    Rotation=Get slope_foot_rot() -> (slope_foot_rot:Rotator)
    Scale="1,1,1"
   ) -> (Pose) {
    alphaInputType=Bool
    rotationMode=BMM_Replace
    rotationSpace=BCS_WorldSpace
    translationMode=BMM_Replace
    translationSpace=BCS_WorldSpace
    alphaBoolBlend {blendOption=Linear}
    boneToModify {boneName=dyn_slope_foot_c}
   }
   n21 = "Transform (Modify) Bone - Bone: dyn_slope_foot_c"(
    ComponentPose=@n20
    bAlphaBoolEnabled=OR Boolean(A=Get bIsSurfaceSwimming(), B=Get bIsDBNOCarried())
    Translation="0,0,-25"
    Rotation="0,0,0"
    Scale="1,1,1"
   ) {
    alphaInputType=Bool
    rotationMode=BMM_Replace
    rotationSpace=BCS_WorldSpace
    translationMode=BMM_Additive
    translationSpace=BCS_BoneSpace
    alphaBoolBlend {blendInTime=0.1, blendOption=Linear, blendOutTime=0.1}
    boneToModify {boneName=dyn_slope_foot_c}
    translation {z=-25}
   }
   Save cached pose 'Pre Leg Dynamics'(Pose=Component To Local(ComponentPose=@n21)) {cacheName=Pre Leg Dynamics}
   n22 = RigidBody(ComponentPose=Local To Component(LocalPose=Use cached pose 'Pre Leg Dynamics'())) {
    bClampLinearTranslationLimitToRefPose=true
    bOverrideWorldGravity=true
    overridePhysicsAsset=&/M_MED_Ruckus_Mini_FEIdle_Physics
    simulationSpace=BaseBoneSpace
    alphaBoolBlend {blendOption=Linear}
    baseBoneRef {boneName=root}
    componentLinearAccScale=<1 1 1>
    componentLinearVelScale=<1 1 1>
    overrideWorldGravity {z=-920}
   }
   n23 = Component To Local(ComponentPose=@n22)
   n24 = RigidBody!(
    ComponentPose=Local To Component(LocalPose=Use cached pose 'Pre Leg Dynamics'())
    bAlphaBoolEnabled:bool=Get vehicleEmoteRBAN() -> (vehicleEmoteRBAN:bool)
    ExternalForce
   ) -> (Pose) {
    alphaInputType=Bool
    bAlphaBoolEnabled=false
    bClampLinearTranslationLimitToRefPose=true
    bOverrideWorldGravity=true
    overridePhysicsAsset=&/M_MED_Ruckus_Mini_Physics
    simulationSpace=BaseBoneSpace
    alphaBoolBlend {blendOption=Linear}
    baseBoneRef {boneName=root}
    componentLinearAccScale=<1 1 1>
    componentLinearVelScale=<1 1 1>
    overrideWorldGravity {z=-920}
   }
   n25 = Component To Local(ComponentPose=@n24)
   n26 = "Transform (Modify) Bone - Bone: d_thigh_r"(
    ComponentPose=Local To Component(LocalPose=Use cached pose 'Pre Leg Dynamics'())
    Alpha=1
    Translation="0,0,0"
    Rotation="0,-12,0"
    Scale="1,1,1"
   ) {
    rotationMode=BMM_Additive
    rotationSpace=BCS_BoneSpace
    translationSpace=BCS_ParentBoneSpace
    alphaBoolBlend {blendOption=Linear}
    boneToModify {boneName=d_thigh_r}
    rotation {yaw=-12}
   }
   n27 = "Transform (Modify) Bone - Bone: d_thigh_l"(ComponentPose=@n26, Alpha=1, Translation="0,0,0", Rotation="0,-12,0", Scale="1,1,1") {
    rotationMode=BMM_Additive
    rotationSpace=BCS_BoneSpace
    translationSpace=BCS_ParentBoneSpace
    alphaBoolBlend {blendOption=Linear}
    boneToModify {boneName=d_thigh_l}
    rotation {yaw=-12}
   }
   n28 = RigidBody!(ComponentPose=@n27, Alpha:float, ExternalForce) -> (Pose) {
    bClampLinearTranslationLimitToRefPose=true
    bOverrideWorldGravity=true
    overridePhysicsAsset=&/M_MED_Ruckus_Mini_Crouch_Physics
    simulationSpace=BaseBoneSpace
    alphaBoolBlend {blendOption=Linear}
    baseBoneRef {boneName=root}
    componentLinearAccScale=<1 1 1>
    componentLinearVelScale=<1 1 1>
    overrideWorldGravity {z=-920}
   }
   n29 = Component To Local(ComponentPose=@n28)
   n30 = Get crouch_or_DBNO() -> (crouch_or_DBNO:bool)
   n31 = "Transform (Modify) Bone - Bone: d_thigh_r"(
    ComponentPose=Local To Component(LocalPose=Blend Poses by bool(BlendPose_0=@n29, BlendPose_1=@n25, BlendTime_0=0.2, BlendTime_1=0.2, bActiveValue=@n30) {blendType=Linear})
    Alpha=1
    Translation="-2.7,3,5.8"
    Rotation="0,0,0"
    Scale="1,1,1"
   ) {
    rotationSpace=BCS_BoneSpace
    translationMode=BMM_Replace
    translationSpace=BCS_ParentBoneSpace
    alphaBoolBlend {blendOption=Linear}
    boneToModify {boneName=d_thigh_r}
    translation=<-2.7 3 5.8>
   }
   n32 = "Transform (Modify) Bone - Bone: d_thigh_l"(ComponentPose=@n31, Alpha=1, Translation="-2.7,3,-5.8", Rotation="0,0,0", Scale="1,1,1") {
    rotationSpace=BCS_BoneSpace
    translationMode=BMM_Replace
    translationSpace=BCS_ParentBoneSpace
    alphaBoolBlend {blendOption=Linear}
    boneToModify {boneName=d_thigh_l}
    translation=<-2.7 3 -5.8>
   }
   n33 = "Transform (Modify) Bone - Bone: d_calf_r"(ComponentPose=@n32, Alpha=1, Translation="22.3,0,0", Rotation="0,0,0", Scale="1,1,1") {
    translationMode=BMM_Replace
    translationSpace=BCS_ParentBoneSpace
    alphaBoolBlend {blendOption=Linear}
    boneToModify {boneName=d_calf_r}
    translation {x=22.3}
   }
   n34 = "Transform (Modify) Bone - Bone: d_calf_l"(ComponentPose=@n33, Alpha=1, Translation="-22.3,0,0", Rotation="0,0,0", Scale="1,1,1") {
    translationMode=BMM_Replace
    translationSpace=BCS_ParentBoneSpace
    alphaBoolBlend {blendOption=Linear}
    boneToModify {boneName=d_calf_l}
    translation {x=-22.3}
   }
   n35 = Component To Local(ComponentPose=@n34)
   n36 = Get PlayingFrontendIdle()
   n37 = Random Sequence Player!() -> (Pose:Pose Link Structure) {
    is=title AnimGraphNode_RandomPlayer
    entries [
     {
      chanceToPlay=0.8
      maxLoopCount=0
      maxPlayRate=1
      minLoopCount=0
      minPlayRate=1
      sequence=/BRCosmetics/Characters/Player/Male/Medium/Bodies/M_MED_Ruckus_Mini/Animations/HeroSkin_Ruckus_Mini_Idle
      blendIn {blendOption=Linear, blendTime=0.2, customCurve=null}
     }
     {
      chanceToPlay=0.2
      maxLoopCount=0
      maxPlayRate=1
      minLoopCount=0
      minPlayRate=1
      sequence=/BRCosmetics/Characters/Player/Male/Medium/Bodies/M_MED_Ruckus_Mini/Animations/HeroSkin_Ruckus_Mini_IdleBreak
      blendIn {blendOption=Linear, blendTime=0.2, customCurve=null}
     }
    ]
   }
   Save cached pose 'Pre ShldrPad Rigging'(Pose=Layered blend per bone(BasePose=Blend Poses by bool(BlendPose_0=@n23, BlendPose_1=@n35, bActiveValue=@n36) {blendType=Linear}, BlendPoses_0=@n37) {layerSetup [{branchFilters [{boneName=d_clavicle_r}, {boneName=d_clavicle_l}]}]}) {cacheName=Pre ShldrPad Rigging}
   Save cached pose 'Copied Pose'(Pose=Slot 'DefaultSlot'(Source=Copy Pose From Mesh() {bUseAttachedParent=true})) {cacheName=Copied Pose}
   n38 = Control Rig!(Source=Use cached pose 'Pre ShldrPad Rigging'(), Alpha=Get weapon_ready() -> (weapon_ready:double)) -> (Pose) {
    alphaBoolBlend {blendOption=Linear}
    controlRigAssetReference {blueprintRigClass=&/M_MED_Ruckus_Mini_head_Clip_CtrlRig.M_MED_Ruckus_Mini_head_Clip_CtrlRig_C}
   }
   n39 = Control Rig(Source=@n38, Alpha=Get Shoulder ctrl() -> (Shoulder ctrl:double)) {
    alphaBoolBlend {blendOption=Linear}
    controlRigAssetReference {blueprintRigClass=&/M_MED_Ruckus_Mini_shoulder_CtrlRig.M_MED_Ruckus_Mini_shoulder_CtrlRig_C}
   }
   n40 = Control Rig!(Source=@n39, bAlphaBoolEnabled:bool=Get Rifle Active() -> (Rifle Active:bool)) -> (Pose) {
    alphaInputType=Bool
    alphaBoolBlend {blendInTime=0.1, blendOption=Linear, blendOutTime=0.1}
    controlRigAssetReference {blueprintRigClass=&/M_MED_Ruckus_Mini_Rifle_CtrlRig.M_MED_Ruckus_Mini_Rifle_CtrlRig_C}
   }
   n41 = Get bIsSkyDiving()
   n42 = "Transform (Modify) Bone - Bone: shldr_l"!(
    ComponentPose=Local To Component(LocalPose=@n40)
    bAlphaBoolEnabled:bool=@n41
    Translation="3,0,7"
    Rotation="0.000,0.000,-45"
    Scale="1,1,1"
   ) -> (Pose) {
    alphaInputType=Bool
    rotationMode=BMM_Additive
    translationMode=BMM_Additive
    alphaBoolBlend {blendInTime=0.3, blendOption=Linear, blendOutTime=0.3}
    boneToModify {boneName=shldr_l}
    rotation {roll=-45}
    translation {x=3, z=7}
   }
   n43 = "Transform (Modify) Bone - Bone: shldr_r"!(ComponentPose=@n42, bAlphaBoolEnabled:bool=@n41, Translation="-3,0,7", Rotation="0.000,0.000,-45", Scale="1,1,1") -> (Pose) {
    alphaInputType=Bool
    rotationMode=BMM_Additive
    translationMode=BMM_Additive
    alphaBoolBlend {blendInTime=0.3, blendOption=Linear, blendOutTime=0.3}
    boneToModify {boneName=shldr_r}
    rotation {roll=-45}
    translation {x=-3, z=7}
   }
   n44 = Get skydive chair() -> (skydive chair:double)
   n45 = "Transform (Modify) Bone - Bone: chair_spine_03"(ComponentPose=@n43, Alpha=@n44, Translation="0,-2,0", Rotation="0,0,-10") {
    rotationMode=BMM_Additive
    translationMode=BMM_Additive
    alphaBoolBlend {blendInTime=0.3, blendOption=Linear, blendOutTime=0.5}
    boneToModify {boneName=chair_spine_03}
    rotation {roll=-10}
    translation {y=-2}
   }
   n46 = "Transform (Modify) Bone - Bone: chair_spine_04"(ComponentPose=@n45, Alpha=@n44, Translation="0,-3,0", Rotation="0.000,0.000,-7") {
    rotationMode=BMM_Additive
    translationMode=BMM_Additive
    alphaBoolBlend {blendInTime=0.3, blendOption=Linear, blendOutTime=0.5}
    boneToModify {boneName=chair_spine_04}
    rotation {roll=-7}
    translation {y=-3}
   }
   n47 = "Transform (Modify) Bone - Bone: chair_spine_05"(ComponentPose=@n46, Alpha=@n44, Translation="0,-3,0", Rotation="0.000,0.000,-7") {
    rotationMode=BMM_Additive
    translationMode=BMM_Additive
    alphaBoolBlend {blendInTime=0.3, blendOption=Linear, blendOutTime=0.5}
    boneToModify {boneName=chair_spine_05}
    rotation {roll=-7}
    translation {y=-3}
   }
   n48 = "Transform (Modify) Bone - Bone: chair_spine_03"(ComponentPose=@n47, Alpha=1, Translation="0,-1.5,0", Rotation="0,-10,0") {
    rotationMode=BMM_Additive
    rotationSpace=BCS_BoneSpace
    translationMode=BMM_Additive
    translationSpace=BCS_BoneSpace
    alphaBoolBlend {blendInTime=0.3, blendOption=Linear, blendOutTime=0.5}
    boneToModify {boneName=chair_spine_03}
    rotation {yaw=-10}
    translation {y=-1.5}
   }
   n49 = "Transform (Modify) Bone - Bone: chair_spine_01"!(ComponentPose=@n48, Alpha=1, Translation="0,-2,0", Rotation="0,0,0") -> (Pose) {
    rotationMode=BMM_Additive
    rotationSpace=BCS_BoneSpace
    translationMode=BMM_Additive
    translationSpace=BCS_BoneSpace
    alphaBoolBlend {blendInTime=0.3, blendOption=Linear, blendOutTime=0.5}
    boneToModify {boneName=chair_spine_01}
    translation {y=-2}
   }
   n50 = "Transform (Modify) Bone - Bone: shldr_r"(ComponentPose=@n49, bAlphaBoolEnabled=True, Translation="-2,-1,0", Rotation="0.000,-15,0.000", Scale="1,1,1") {
    alphaInputType=Bool
    rotationMode=BMM_Additive
    translationMode=BMM_Additive
    alphaBoolBlend {blendInTime=0.3, blendOption=Linear, blendOutTime=0.1}
    boneToModify {boneName=shldr_r}
    rotation {yaw=-15}
    translation=<-2 -1>
   }
   n51 = "Transform (Modify) Bone - Bone: shldr_l"(ComponentPose=@n50, bAlphaBoolEnabled=True, Translation="2,-1,0", Rotation="0,0,0", Scale="1,1,1") {
    alphaInputType=Bool
    translationMode=BMM_Additive
    alphaBoolBlend {blendInTime=0.3, blendOption=Linear, blendOutTime=0.1}
    boneToModify {boneName=shldr_l}
    translation=<2 -1>
   }
   n52 = Get shoulder_override() -> (shoulder_override:double)
   n53 = "Constraint - chair_spine_03 (R)"(ComponentPose=@n51, Alpha=@n52, ConstraintWeights_0=1) {
    alphaBoolBlend {blendOption=Linear}
    alphaScaleBiasClamp {scale=0.1}
    boneToModify {boneName=chair_spine_03}
    constraintSetup [{transformType=Rotation, targetBone {boneName=spine_03}}]
    constraintWeights [1]
   }
   n54 = "Constraint - chair_spine_04 (R)"(ComponentPose=@n53, Alpha=@n52, ConstraintWeights_0=1) {
    alphaBoolBlend {blendOption=Linear}
    alphaScaleBiasClamp {scale=0.1}
    boneToModify {boneName=chair_spine_04}
    constraintSetup [{transformType=Rotation, targetBone {boneName=spine_04}}]
    constraintWeights [1]
   }
   n55 = "Constraint - chair_spine_05 (R)"(ComponentPose=@n54, Alpha=@n52, ConstraintWeights_0=1) {
    alphaBoolBlend {blendOption=Linear}
    alphaScaleBiasClamp {scale=0.3}
    boneToModify {boneName=chair_spine_05}
    constraintSetup [{transformType=Rotation, targetBone {boneName=spine_05}}]
    constraintWeights [1]
   }
   n56 = "Constraint - shldr_r (P)"(ComponentPose=@n55, Alpha=@n52, ConstraintWeights_0=1) {
    alphaBoolBlend {blendOption=Linear}
    boneToModify {boneName=shldr_r}
    constraintSetup [{targetBone {boneName=chair_spine_05}}]
    constraintWeights [1]
   }
   n57 = "Constraint - shldr_l (P)"(ComponentPose=@n56, Alpha=@n52, ConstraintWeights_0=1) {
    alphaBoolBlend {blendOption=Linear}
    boneToModify {boneName=shldr_l}
    constraintSetup [{targetBone {boneName=chair_spine_05}}]
    constraintWeights [1]
   }
   n58 = "Transform (Modify) Bone - Bone: chair_spine_01"(ComponentPose=@n57, Alpha=1, Translation="0,-2,0", Rotation="0,0,0") {
    rotationMode=BMM_Additive
    rotationSpace=BCS_BoneSpace
    translationMode=BMM_Additive
    translationSpace=BCS_BoneSpace
    alphaBoolBlend {blendInTime=0.3, blendOption=Linear, blendOutTime=0.5}
    alphaScaleBiasClamp {bInterpResult=true, interpSpeedDecreasing=2}
    boneToModify {boneName=chair_spine_01}
    translation {y=-2}
   }
   n59 = "Transform (Modify) Bone - Bone: chair_spine_04"(ComponentPose=@n58, Alpha=1, Translation="0,-3,0", Rotation="0,0,0") {
    rotationMode=BMM_Additive
    rotationSpace=BCS_BoneSpace
    translationMode=BMM_Additive
    translationSpace=BCS_BoneSpace
    alphaBoolBlend {blendInTime=0.3, blendOption=Linear, blendOutTime=0.5}
    boneToModify {boneName=chair_spine_04}
    translation {y=-3}
   }
   n60 = "Transform (Modify) Bone - Bone: chair_spine_04"(ComponentPose=@n59, Alpha=1, Translation="-0.000,-3,0", Rotation="0,0,0") {
    rotationMode=BMM_Additive
    rotationSpace=BCS_BoneSpace
    translationMode=BMM_Additive
    translationSpace=BCS_BoneSpace
    alphaBoolBlend {blendInTime=0.3, blendOption=Linear, blendOutTime=0.5}
    boneToModify {boneName=chair_spine_04}
    translation {y=-3}
   }
   n61 = Control Rig!(Source=Component To Local(ComponentPose=@n60), Alpha:float) -> (Pose) {
    alphaBoolBlend {blendOption=Linear}
    controlRigAssetReference {blueprintRigClass=&/M_MED_Ruckus_Mini_Shldr_Override_CtrlRig.M_MED_Ruckus_Mini_Shldr_Override_CtrlRig_C}
   }
   Save cached pose 'Shldr Pad rigging applied'(Pose=@n61) {cacheName=Shldr Pad rigging applied}
  ]}
 EventGraph {nodes [
   n1 = Event BlueprintUpdateAnimation()
   > n2 = Sequence()
   n3 = Sequence!(execute=@n2.then_1) -> (
    then_0
    then_1
    then_2:Exec
    then_3:Exec
    then_4:Exec
    then_5:Exec
    then_6:Exec
    then_7:Exec
    then_8:Exec
    then_9:Exec
    then_10:Exec
    then_11:Exec
    then_12:Exec
    then_13:Exec
   )
   n4 = Break Fort Anim Input Slope Info!(FortAnimInput_SlopeInfo:"Fort Anim Input Slope Info Structure (by ref)"=Get SlopeInfoInput() -> (SlopeInfoInput:Fort Anim Input Slope Info Structure)) -> (RootSlopeTranslation:Vector=~, LeftFootSlopeTranslation:Vector=~, RightFootSlopeTranslation:Vector=~, SlopeRotation:Rotator=~) {is=title K2Node_BreakStruct}
   n5 = BreakVector(InVec=@n4)
   n6 = BreakVector(InVec=VInterpTo!(Current:Vector=Get slope_foot_trans(), Target:Vector=@n4, DeltaTime:float=@n1.DeltaTimeX, InterpSpeed:float=4) -> (ReturnValue:Vector))
   > Set slope_foot_trans(slope_foot_trans:Vector=MakeVector(
     X=@n5
     Y=@n5.Y
     Z="Min (Float)"(A:double=@n6.Z, B:double=@n5.Z) -> (ReturnValue:double) {is=title K2Node_CommutativeAssociativeBinaryOperator}
    )) -> (Output_Get:Vector)
   > Set slope_foot_rot(slope_foot_rot:Rotator=RInterpTo!(Current:Rotator=Get slope_foot_rot(), Target:Rotator=@n4.SlopeRotation, DeltaTime:float=@n1.DeltaTimeX, InterpSpeed:float=2) -> (ReturnValue:Rotator)) -> (Output_Get:Rotator)
   Set head_back(
    execute=@n3.then_1
    head_back:double=GetAlphaFromReferenceTransformDelta(
     SkeletalMeshComponent=Get skelmesh() -> (skelmesh:Skeletal Mesh Component Object Reference)
     BoneName=head
     BaseName=spine_05
     DeltaComponentType=QuaternionTwist
     InputRange=(X=0,Y=45)
     TwistAxis="0,0,1"
    )
   ) -> (Output_Get:double)
   n7 = Get Anim_Inst() -> (Anim_Inst:Anim Instance Object Reference)
   Set dumbbell(
    execute=@n3.then_10
    dumbbell:double="To Float (Boolean)"(InBool=Montage_IsPlaying(self=@n7, Montage=/BRCosmetics/Animation/Game/MainPlayer/Emotes/Dumbbell/Emote_Dumbbell_Lift_CMM_M))
   ) -> (Output_Get:double)
   n8 = Get bIsDBNO()
   n9 = Set crouch_or_DBNO(execute=@n3.then_11, crouch_or_DBNO:bool=OR Boolean(A=Get bIsCrouching(), B=@n8)) -> (Output_Get:bool)
   > Set crouch_or_DBNO_alpha(crouch_or_DBNO_alpha:double="To Float (Boolean)"(InBool=@n9)) -> (Output_Get:double)
   > Set dbno_alpha(dbno_alpha:double="To Float (Boolean)"(InBool=@n8)) -> (Output_Get:double)
   Set ADS(execute=@n3.then_12, ADS:double="To Float (Boolean)"(InBool=Get bIsTargeting())) -> (Output_Get:double)
   n10 = Get bIsTargeting()
   n11 = Get bIsCrouching()
   > Set crouchADS(crouchADS:double="To Float (Boolean)"(InBool=AND Boolean(A=@n11, B=@n10))) -> (Output_Get:double)
   n12 = Get bIsPlayingMeleeAnim()
   n13 = Get bIsMoving2D()
   > Set Melee Move(Melee Move:double=SelectFloat(A=1, B=0, bPickA=AND Boolean(A=@n12, B=@n13))) -> (Output_Get:double)
   n14 = GetCurrentActiveMontage!(self:Anim Instance Object Reference=GetAnimInstance(self=GetOwnerSkeletalMeshComponent())) -> (ReturnValue:Anim Montage Object Reference)
   > Set Pickaxe_Equip(Pickaxe_Equip:double=SelectFloat(
     A=1
     B=0
     bPickA="Equal (Object)"(A=@n14, B=/Game/Animation/Game/MainPlayer/Combat/Melee/DualWield/Fists/Fists_DualWield_Melee_Equip_CMM_M) {is=title K2Node_CallFunction}
    )) -> (Output_Get:double)
   > Set Not Custom Anim(Not Custom Anim:bool=NOT Boolean(A=Get bCanPlayCustomAnimations() -> (bCanPlayCustomAnimations:bool))) -> (Output_Get:bool)
   n15 = Branch(
    execute=@n3.then_2
    Condition=OR Boolean!(A=Get bIsRelaxedLevel1() -> (bIsRelaxedLevel1:bool), B=Get bIsRelaxedLevel2() -> (bIsRelaxedLevel2:bool), C:bool=Get bIsSkyDiving()) -> (ReturnValue)
   )
   > Set weapon_ready(weapon_ready:double=0) -> (Output_Get:double)
   Set weapon_ready(
    execute=[@n16.then, @n15.else]
    weapon_ready=GetAlphaFromReferenceTransformDelta(
     SkeletalMeshComponent=Get skelmesh()
     BoneName=head
     BaseName=pelvis
     DeltaComponentType=QuaternionTwist
     InputRange=(X=0,Y=55)
     TwistAxis="0,0,1"
    )
   )
   n16 = Branch(execute=@n3.then_3, Condition=Get bIsTargeting())
   n17 = Cast To FortPlayerAnimInstance(execute=@n3.then_4, Object=GetAnimInstance(self=GetOwnerSkeletalMeshComponent()))
   > Set Pistol(Pistol:bool=false) -> (Output_Get:bool)
   > Set Rifle(Rifle:bool=false) -> (Output_Get:bool)
   > n18 = Switch on EFortWeaponCoreAnimation!(execute, Selection:EFortWeaponCoreAnimation Enum=Get WeaponCoreAnim(self=@n17)) -> (
    Melee:Exec
    Pistol:Exec
    Shotgun:Exec
    PaperBlueprint:Exec
    Rifle:Exec
    MeleeOneHand:Exec
    MachinePistol:Exec
    RocketLauncher:Exec
    GrenadeLauncher:Exec
    GoingCommando:Exec
    AssaultRifle:Exec
    TacticalShotgun:Exec
    SniperRifle:Exec
    TrapPlacement:Exec
    ShoulderLauncher:Exec
    AbilityDecoTool:Exec
    Crossbow:Exec
    C4:Exec
    RemoteControl:Exec
    DualWield:Exec
    AR_BullPup:Exec
    AR_ForwardGrip:Exec
    MedPackPaddles:Exec
    SMG_P90:Exec
    AR_DrumGun:Exec
    Consumable_Small:Exec
    Consumable_Large:Exec
    Balloon:Exec
    MountedTurret:Exec
    CreativeTool:Exec
    ExplosiveBow:Exec
    AshtonIndigo:Exec
    AshtonChicago:Exec
    MeleeDualWield:Exec
    Unarmed:Exec
   )
   Branch(execute=[
     @n18.AssaultRifle
     @n18.Crossbow
     @n18.GrenadeLauncher
     @n18.Rifle
     @n18.RocketLauncher
     @n18.Shotgun
     @n18.SniperRifle
     @n18.TacticalShotgun
    ])
   > Set Rifle(Rifle=true)
   Branch(execute=[@n18.MachinePistol, @n18.Pistol])
   > Set Pistol(Pistol=true)
   n19 = Branch(execute=@n3.then_5, Condition=OR Boolean(A=Get bIsSkyDiving(), B=Get bIsSprinting(), C=Get bIsCrouchSprinting()))
   > Set Shoulder ctrl(Shoulder ctrl:double=0) -> (Output_Get:double)
   Set Shoulder ctrl(execute=@n19.else, Shoulder ctrl=1)
   n20 = Branch(
    execute=@n3.then_6
    Condition=OR Boolean!(A=Get bIsRelaxedLevel1(), B=Get bIsRelaxedLevel2(), C=Get bIsSprinting(), D:bool=Get bIsCrouchSprinting()) -> (ReturnValue)
   )
   Set Rifle Active(execute=[@n22.else, @n20.then], Rifle Active:bool=false) -> (Output_Get:bool)
   n21 = Get bIsSkyDiving()
   n22 = Branch(execute=@n20.else, Condition=AND Boolean(A=Get Rifle() -> (Rifle:bool), B=NOT Boolean(A=@n21)))
   Set Rifle Active(execute=[@n22.then, @n23.then], Rifle Active=true)
   n23 = Branch(execute=@n3.then_7, Condition=AND Boolean(A=Get Rifle(), B=Get bIsTargeting()))
   n24 = Branch(execute=@n3.then_8, Condition=Get bIsSkyDiving())
   > Set skydive chair(skydive chair:double=GetAlphaFromReferenceTransformDelta(
     SkeletalMeshComponent=Get skelmesh()
     BoneName=d_spine_05
     BaseName=d_pelvis
     DeltaComponentType=QuaternionTwist
     InputRange=(X=10,Y=20)
     TwistAxis="0,0,1"
    )) -> (Output_Get:double)
   Set skydive chair(execute=@n24.else, skydive chair=0)
   n25 = Cast To FortPlayerAnimInstance(execute=@n3.then_9, Object=GetAnimInstance(self=GetOwnerSkeletalMeshComponent()))
   > Set weapon(weapon:String=Enum to String!(Enumerator:EFortWeaponCoreAnimation Enum=Get WeaponCoreAnim(self=@n25)) -> (ReturnValue:String)) -> (Output_Get:String)
   Set vehicleEmoteRBAN(execute=@n3.then_13, vehicleEmoteRBAN:bool=true) -> (Output_Get:bool)
   n26 = GetAnimInstance(self=GetOwnerSkeletalMeshComponent())
   > Is Valid(InputObject=@n26)
   > n27 = For Each Loop with Break!(
    Exec:Exec
    Array:Array of Anim Montage Object References=Get Vehicle_Montages() -> (Vehicle_Montages:Array of Anim Montage Object References)
    Break:Exec=@n28.then
   ) -> (LoopBody:Exec, Array Element:Anim Montage Object Reference, Array Index:int, Completed:Exec)
   > Branch(Condition=Montage_IsPlaying(self=@n26, Montage=@n27))
   > n28 = Set vehicleEmoteRBAN(vehicleEmoteRBAN=false)
   Get!(execute:Exec=@n2.then_0) -> (then:Exec, else:Exec, Owner_Skel_Mesh:Skeletal Mesh Component Object Reference)
   > Get!(execute) -> (then, else, Anim_Inst:Anim Instance Object Reference)
   n29 = Get Anim_Inst()
   > Set PlayingFrontendIdle(PlayingFrontendIdle:bool="float > float"(A=GetCurveValue(self=@n29, CurveName=FrontEnd_Idle_Bypass), B=0) {is=title K2Node_CallFunction}) -> (Output_Get:bool)
   > Set FrontEndIdle(FrontEndIdle:double=SelectFloat(
     A=1
     B=0
     bPickA=Montage_IsActive!(
      self:Anim Instance Object Reference=@n29
      Montage:Anim Montage Object Reference=/BRCosmetics/Animation/Game/MainPlayer/Menu/BR/CustomSkins/_S17/Ruckus_Mini/Ruckus_Mini_Male_Idle_Montage
     ) -> (ReturnValue:bool)
    )) -> (Output_Get:double)
   > Set FrontEndIdlePosition(FrontEndIdlePosition:double=Montage_GetPosition!(
     self:Anim Instance Object Reference=@n29
     Montage:Anim Montage Object Reference=&Ruckus_Mini_Male_Idle_Montage
    ) -> (ReturnValue:float)) -> (Output_Get:double)
   Event BlueprintInitializeAnimation()
   > Set skelmesh(skelmesh:Skeletal Mesh Component Object Reference=GetPartSkeletalMeshComponent()) -> (Output_Get:Skeletal Mesh Component Object Reference)
   n30 = GetOwnerSkeletalMeshComponent()
   > Is Valid(InputObject=@n30)
   > n31 = Set Owner_Skel_Mesh(Owner_Skel_Mesh:Skeletal Mesh Component Object Reference=@n30) -> (Output_Get:Skeletal Mesh Component Object Reference)
   > Set Anim_Inst(Anim_Inst:Anim Instance Object Reference=GetAnimInstance(self=@n31)) -> (Output_Get:Anim Instance Object Reference)
   GetVectorArrayAverage!(Vectors:Array of Vectors=Make Array!("[0]":Vector=@n4.LeftFootSlopeTranslation, "[1]":Vector=@n4.RightFootSlopeTranslation) -> (Array:Array of Vectors)) -> (ReturnValue:Vector)
  ]}
]
variables [
 ADS {type=number}
 Anim_Inst {type=AnimInstance, default=null}
 crouch_or_DBNO {type=boolean}
 crouch_or_DBNO_alpha {type=number}
 crouchADS {type=number}
 dbno_alpha {type=number}
 dumbbell {type=number}
 FrontEndIdle {type=number}
 FrontEndIdlePosition {type=number}
 head_back {type=number}
 Melee Move {type=number}
 Not Custom Anim {type=boolean}
 Owner_Skel_Mesh {type=SkeletalMeshComponent}
 Pickaxe_Equip {type=number}
 Pistol {type=boolean}
 PlayingFrontendIdle {type=boolean}
 Rifle {type=boolean}
 Rifle Active {type=boolean}
 Shoulder ctrl {type=number}
 shoulder_override {type=number, default=0.3}
 skelmesh {type=SkeletalMeshComponent}
 skydive chair {type=number}
 slope_foot_rot {type=Rotator}
 slope_foot_trans {type=Vector}
 Vehicle_Montages {
  type="array<AnimMontage>"
  default [
   /BRCosmetics/Animation/Game/MainPlayer/Emotes/Jupiter/Emote_Jupiter_CMM_M
   /BRCosmetics/Animation/Game/MainPlayer/Emotes/Donut1/Donut_One_CMM_M
   /BRCosmetics/Animation/Game/MainPlayer/Emotes/Lifted/Emote_Lifted_CMM_M
   /BRCosmetics/Animation/Game/MainPlayer/Emotes/Roving/Emote_Roving_CMM_M
   /BRCosmetics/Animation/Game/MainPlayer/Emotes/Saucer/Emote_Saucer_CMM_M
   /BRCosmetics/Animation/Game/MainPlayer/Emotes/Chugga/Emote_Chugga_Train_CMM_M
   /BRCosmetics/Animation/Game/MainPlayer/Emotes/Jingle/Emote_jingle_CMM_M
   /BRCosmetics/Animation/Game/MainPlayer/Emotes/Hydraulics/Emote_Hydraulics_CMM_M
   /BRCosmetics/Animation/Game/MainPlayer/Emotes/Office/Emote_Office_CMM_M
   /BRCosmetics/Animation/Game/MainPlayer/Emotes/Hilda/Emote_Hilda_CMM_M
   /BRCosmetics/Animation/Game/MainPlayer/Emotes/Lifted/Emote_Lifted_CMM_M_Frontend
   /BRCosmetics/Animation/Game/MainPlayer/Emotes/Donut1/Donut_One_CMM_M_Frontend
   /BRCosmetics/Animation/Game/MainPlayer/Emotes/Jingle/Emote_jingle_frontend_CMM_M
   /BRCosmetics/Animation/Game/MainPlayer/Emotes/Roving/Emote_Roving_CMM_Frontend_M
   /BRCosmetics/Animation/Game/MainPlayer/Emotes/Chugga/Emote_Chugga_Train_CMM_Frontend_M
   /BRCosmetics/Animation/Game/MainPlayer/Emotes/Chugga/Emote_Chugga_Passenger_CMM_M
   /BRCosmetics/Animation/Game/MainPlayer/Emotes/Hydraulics/Emote_Hydraulics_CMM_M_FE
  ]
 }
 vehicleEmoteRBAN {type=boolean}
 weapon {type=string, default=""}
 weapon_ready {type=number}
]
