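# Control Rig for the Ruckus Mini player face accessory, fitting the face onto a secondary "d_" spine chain.
# It aims torso (half weight) and d_spine_01 toward neck_02, blends d_chest_end toward head rotation, then re-seats faceAttach on d_chest_end at its initial offset.
# A separate branch, not wired to Begin Execution, scales each browLong_hide bone by its outer brow's displacement from rest, using a clamped remap.
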ControlRigBlueprint /BRCosmetics/Characters/Player/Male/Medium/Bodies/M_MED_Ruckus_Mini/Meshes/Parts/M_MED_Ruckus_Mini_FaceAcc_CtrlRig
graphs [
 RigVMFunctionLibrary {links [], nodes []}
 RigVMModel {nodes [
   RigUnit_GetTransform = &RigUnit_GetTransform(Item=(Type=Bone,Name="L_brow_outer"), Space=LocalSpace, bInitial=true)
   n1 = /Script/RigVM.RigVMFunction_MathVectorLength(Value=/Script/RigVM.RigVMFunction_MathVectorSub(
     B=@RigUnit_GetTransform.Transform.Translation
     A=&RigUnit_GetTransform(Item=(Type=Bone,Name="L_brow_outer"), Space=LocalSpace).Transform.Translation
    ) {is=struct RigVMUnitNode, template="Subtract::Execute(in A,in B,out Result)"}.Result) {is=struct RigVMUnitNode, template="Length::Execute(in Value,out Result)"}
   &RigUnit_OffsetTransformForItem(
    ExecutePin=()
    Item=(Type=Bone,Name="L_browLong_hide")
    OffsetTransform=(Rotation=(X=0,Y=0,Z=0,W=1),Translation=(X=0,Y=0,Z=0),Scale3D=(X=1,Y=1,Z=1))
    Weight=1
    bPropagateToChildren=True
    OffsetTransform.Scale3D=/Script/RigVM.RigVMFunction_MathVectorFromFloat(Value=&RigVMFunction_MathFloatRemap(SourceMaximum=6, SourceMinimum=3, TargetMinimum=0.1, bClamp=true, Value=@n1.Result).Result) {is=struct RigVMUnitNode, template="Cast::Execute(in Value,out Result)"}.Result
   )
   n2 = &RigVMFunction_MathVectorLength(Value=&RigVMFunction_MathVectorSub(
     A=&RigUnit_GetTransform(Item=(Type=Bone,Name="R_brow_outer"), Space=LocalSpace).Transform.Translation
     B=&RigUnit_GetTransform(Item=(Type=Bone,Name="R_brow_outer"), Space=LocalSpace, bInitial=true).Transform.Translation
    ) {template="Subtract::Execute(in A,in B,out Result)"}.Result) {template="Length::Execute(in Value,out Result)"}
   > &RigUnit_OffsetTransformForItem(
    Item=(Type=Bone,Name="R_browLong_hide")
    OffsetTransform=(Rotation=(X=0,Y=0,Z=0,W=1),Translation=(X=0,Y=0,Z=0),Scale3D=(X=1,Y=1,Z=1))
    Weight=1
    bPropagateToChildren=True
    OffsetTransform.Scale3D=&RigVMFunction_MathVectorFromFloat(Value=&RigVMFunction_MathFloatRemap(SourceMaximum=6, SourceMinimum=3, TargetMinimum=0.1, bClamp=true, Value=@n2.Result).Result) {template="Cast::Execute(in Value,out Result)"}.Result
   )
   n3 = &RigUnit_BeginExecution()
   Item = /Script/ControlRig.RigUnit_Item(Item=(Type=Bone,Name="torso")) {is=struct RigVMUnitNode}
   RigUnit_GetRelativeTransformForItem = &RigUnit_GetRelativeTransformForItem(Child=(Type=Bone,Name="neck_02"), bChildInitial=true, bParentInitial=true, Parent=@Item.Item)
   /Script/ControlRig.RigUnit_AimItem(
    DebugSettings=(Scale=10,WorldOffset=(Rotation=(X=0,Y=0,Z=0,W=1),Translation=(X=0,Y=0,Z=0),Scale3D=(X=1,Y=1,Z=1)))
    Primary=(Weight=1,Axis=(X=1,Y=0,Z=0),Target=(X=0,Y=0,Z=0),Kind=Location,Space=(Type=Bone,Name="neck_02"))
    Secondary=(Weight=0,Axis=(X=0,Y=0,Z=1),Target=(X=1,Y=0,Z=0),Kind=Location,Space=(Type=Bone))
    Weight=0.5
    Item=@Item.Item
    ExecutePin=@n3.ExecutePin
    Primary.Axis=@RigUnit_GetRelativeTransformForItem.RelativeTransform.Translation
   ) {is=struct RigVMUnitNode}
   n4 = &RigUnit_Item(Item=(Type=Bone,Name="d_spine_01"))
   > &RigUnit_AimItem(
    DebugSettings=(Scale=10,WorldOffset=(Rotation=(X=0,Y=0,Z=0,W=1),Translation=(X=0,Y=0,Z=0),Scale3D=(X=1,Y=1,Z=1)))
    Primary=(Weight=1,Axis=(X=1,Y=0,Z=0),Target=(X=0,Y=0,Z=0),Kind=Location,Space=(Type=Bone,Name="neck_02"))
    Secondary=(Weight=0,Axis=(X=0,Y=0,Z=1),Target=(X=1,Y=0,Z=0),Kind=Location,Space=(Type=Bone))
    Weight=1
    Item=@n4.Item
    Primary.Axis=&RigUnit_GetRelativeTransformForItem(Child=(Type=Bone,Name="spine_01_fr_tgt"), bChildInitial=true, bParentInitial=true, Parent=@n4.Item).RelativeTransform.Translation
   )
   > &RigUnit_SetRotation(
    Item=(Type=Bone,Name="d_chest_end")
    Space=GlobalSpace
    Weight=0.6
    Value=&RigUnit_GetTransform(Item=(Type=Bone,Name="head")).Transform.Rotation
   )
   > &RigUnit_SetTransform(
    Item=(Type=Bone,Name="faceAttach")
    Weight=1
    Value=&RigVMFunction_MathTransformMul(
     A=&RigUnit_GetRelativeTransformForItem(Child=(Type=Bone,Name="faceAttach"), Parent=(Type=Bone,Name="d_chest_end"), bChildInitial=true, bParentInitial=true).RelativeTransform
     B=&RigUnit_GetTransform(Item=(Type=Bone,Name="d_chest_end")).Transform
    ) {template="Multiply::Execute(in A,in B,out Result)"}.Result
   )
  ]}
]
hierarchy [
 root {loc=<0 0 0>, rot=<0 0 0 1>}
 pelvis {parent=root, loc=<0 -0.7071 93.1191>, rot=<0.0127 0.707 -0.0127 -0.707>}
 spine_01 {parent=pelvis, loc=<6.7494 -0.6388 0>, rot=<0 0 0.034 -0.9994>}
 spine_02 {parent=spine_01, loc=<7.2517 -0.035 0>, rot=<0 0 0.0238 -0.9997>}
 spine_03 {parent=spine_02, loc=<6.0104 0 0>, rot=<0 0 0.0149 -0.9999>}
 spine_04 {parent=spine_03, loc=<8.1023 0 0>, rot=<0 0 -0.0132 -0.9999>}
 spine_05 {parent=spine_04, loc=<10.0146 0 0>, rot=<0 0 -0.0015 -1>}
 neck_01 {parent=spine_05, loc=<11.3138 0.1529 0>, rot=<0 0 -0.2003 -0.9797>}
 neck_02 {parent=neck_01, loc=<4.5627 0 0>, rot=<0 0 0.0504 -0.9987>}
 head {parent=neck_02, loc=<5.3914 0 0>, rot=<0 0 0.115 -0.9934>}
 faceAttach {parent=head, loc=<-4.133 -0.9158 0>, rot=<0.0082 -0.7071 -0.0082 -0.7071>, scale=<0.89 0.89 0.89>}
 C_brow_mid {parent=faceAttach, loc=<-0.002 17.4108 9.8976>, rot=<0 0 0 -1>}
 L_brow_mid {parent=faceAttach, loc=<3.3172 16.8427 10.0259>, rot=<-0.0001 -0.0042 0.2553 -0.9669>}
 L_brow_outer {parent=faceAttach, loc=<11.6357 12.2566 5.3967>, rot=<-0.0287 -0.04 0.4491 -0.8921>}
 L_cheek_inner {parent=faceAttach, loc=<4.6389 18.853 -1.8861>, rot=<-0.0706 -0.2697 0.3184 -0.906>}
 C_lip_upper_mid {parent=faceAttach, loc=<0 18.5798 -7.4457>, rot=<0.2225 0.0013 -0.0005 -0.9749>}
 L_lip_upper_outer {parent=faceAttach, loc=<3.1313 18.1331 -7.174>, rot=<0.0359 -0.0728 0.2627 -0.9615>}
 L_lip_corner {parent=faceAttach, loc=<5.0006 16.9255 -7.7206>, rot=<0.0029 0.008 -0.0146 -0.9999>}
 C_nose_bridge {parent=faceAttach, loc=<-0.0198 19.8346 -0.3597>, rot=<-0.2218 0 0 -0.9751>}
 R_brow_mid {parent=faceAttach, loc=<-3.3172 16.8427 10.0259>, rot=<-0.0001 0.0042 -0.2553 -0.9669>}
 R_brow_outer {parent=faceAttach, loc=<-11.6357 12.2566 5.3967>, rot=<-0.0287 0.04 -0.4491 -0.8921>}
 R_cheek_inner {parent=faceAttach, loc=<-4.6389 18.853 -1.8861>, rot=<-0.0706 0.2697 -0.3184 -0.906>}
 R_lip_corner {parent=faceAttach, loc=<-5.0006 16.9255 -7.7206>, rot=<0.0029 -0.008 0.0146 -0.9999>}
 C_jaw {parent=faceAttach, loc=<-0.0151 11.7234 -6.7008>, rot=<0 0 0 1>}
 L_lip_lower_outer {parent=C_jaw, loc=<3.1652 5.9832 -1.1509>, rot=<0.013 0.0445 0.3057 -0.951>}
 C_lip_lower_mid {parent=C_jaw, loc=<0.0151 6.6244 -1.0761>, rot=<0 0 0 -1>}
 R_lip_lower_outer {parent=C_jaw, loc=<-3.1351 5.9832 -1.1509>, rot=<0.013 -0.0445 -0.3057 -0.951>}
 teeth_lower {parent=C_jaw, loc=<0.0151 4.7507 -1.7793>, rot=<0 0 0 1>}
 tongue {parent=C_jaw, loc=<0.0158 0.5525 -1.4869>, rot=<-0.1568 0 0 -0.9876>}
 R_eye {parent=faceAttach, loc=<-6.2293 16.4069 3.9348>, rot=<0 0 0.9998 -0.0187>}
 R_eye_lid_upper_mid {parent=R_eye, loc=<-0.0396 -1.1714 0.2631>, rot=<0.0523 -0.0009 0.0167 -0.9985>}
 R_eye_lid_lower_mid {parent=R_eye, loc=<-0.0085 -0.9425 -0.7051>, rot=<0.0329 -0.0005 0.0167 -0.9993>}
 L_eye {parent=faceAttach, loc=<6.2293 16.4069 3.9348>, rot=<0 0 0.9998 0.0187>}
 L_eye_lid_upper_mid {parent=L_eye, loc=<0.0251 -1.1623 0.2388>, rot=<0.0523 0 0 -0.9986>}
 L_eye_lid_lower_mid {parent=L_eye, loc=<0.0223 -0.923 -0.7488>, rot=<0.0329 0 0 -0.9995>}
 teeth_upper {parent=faceAttach, loc=<0 16.6214 -6.4522>, rot=<0 0 0 1>}
 R_lip_upper_outer {parent=faceAttach, loc=<-3.1313 18.1331 -7.174>, rot=<0.0359 0.0728 -0.2627 -0.9615>}
 L_lid_upr_main {parent=faceAttach, loc=<6.3271 16.4705 4.0371>, rot=<1 0 0 -0.0074>, scale=<5.9825 5.9825 5.9825>}
 L_lid_upr_mid {parent=L_lid_upr_main, loc=<0 0 -0.5>, rot=<0 0 0 1>, scale=<0.866 0.866 1>}
 L_lid_upr_1 {parent=L_lid_upr_main, loc=<0 0 -0.5919>, rot=<0 0 0 1>, scale=<0.8222 0.8222 1>}
 L_lid_upr_2 {parent=L_lid_upr_main, loc=<0 0 -0.6827>, rot=<0 0 0 1>, scale=<0.7891 0.7891 1>}
 L_lid_upr_3 {parent=L_lid_upr_main, loc=<0 0 -0.7672>, rot=<0 0 0 1>, scale=<0.7697 0.7697 1>}
 L_lid_upr_4 {parent=L_lid_upr_main, loc=<0 0 -0.843>, rot=<0 0 0 1>, scale=<0.7585 0.7585 1>}
 L_lid_upr_5 {parent=L_lid_upr_main, loc=<0 0 -0.9071>, rot=<0 0 0 1>, scale=<0.7576 0.7576 1>}
 L_lid_upr_6 {parent=L_lid_upr_main, loc=<0 0 -0.9566>, rot=<0 0 0 1>, scale=<0.7604 0.7604 1>}
 L_lid_upr_end {parent=L_lid_upr_main, loc=<0 0 -1>, rot=<0 0 0 1>}
 L_lid_lwr_main {parent=faceAttach, loc=<6.3271 16.4705 4.0369>, rot=<1 0 0 -0.0074>, scale=<5.9219 5.9219 5.9219>}
 L_lid_lwr_mid {parent=L_lid_lwr_main, loc=<0 0 1>, rot=<0 0 0 1>, scale=<0.05 0.05 1>}
 L_lid_lwr_1 {parent=L_lid_lwr_main, loc=<0 0 1>, rot=<0 0 0 1>, scale=<0.05 0.05 1>}
 L_lid_lwr_2 {parent=L_lid_lwr_main, loc=<0 0 1>, rot=<0 0 0 1>, scale=<0.05 0.05 1>}
 L_lid_lwr_3 {parent=L_lid_lwr_main, loc=<0 0 1>, rot=<0 0 0 1>, scale=<0.05 0.05 1>}
 L_lid_lwr_4 {parent=L_lid_lwr_main, loc=<0 0 1>, rot=<0 0 0 1>, scale=<0.05 0.05 1>}
 L_lid_lwr_5 {parent=L_lid_lwr_main, loc=<0 0 1>, rot=<0 0 0 1>, scale=<0.05 0.05 1>}
 L_lid_lwr_6 {parent=L_lid_lwr_main, loc=<0 0 1>, rot=<0 0 0 -1>, scale=<0.05 0.05 1>}
 L_lid_lwr_end {parent=L_lid_lwr_main, loc=<0 0 1.0001>, rot=<0 0 0 -1>}
 R_lid_upr_main {parent=faceAttach, loc=<-6.3271 16.4705 4.0371>, rot=<0 -0.0074 1 0>, scale=<5.9825 5.9825 5.9825>}
 R_lid_upr_mid {parent=R_lid_upr_main, loc=<0 0 0.5>, rot=<0 0 0 1>, scale=<0.866 0.866 1>}
 R_lid_upr_1 {parent=R_lid_upr_main, loc=<0 0 0.5919>, rot=<0 0 0 1>, scale=<0.8222 0.8222 1>}
 R_lid_upr_2 {parent=R_lid_upr_main, loc=<0 0 0.6827>, rot=<0 0 0 1>, scale=<0.7891 0.7891 1>}
 R_lid_upr_3 {parent=R_lid_upr_main, loc=<0 0 0.7672>, rot=<0 0 0 1>, scale=<0.7697 0.7697 1>}
 R_lid_upr_4 {parent=R_lid_upr_main, loc=<0 0 0.843>, rot=<0 0 0 1>, scale=<0.7585 0.7585 1>}
 R_lid_upr_5 {parent=R_lid_upr_main, loc=<0 0 0.9071>, rot=<0 0 0 -1>, scale=<0.7576 0.7576 1>}
 R_lid_upr_6 {parent=R_lid_upr_main, loc=<0 0 0.9566>, rot=<0 0 0 1>, scale=<0.7604 0.7604 1>}
 R_lid_upr_end {parent=R_lid_upr_main, loc=<0 0 1>, rot=<0 0 0 1>}
 R_lid_lwr_main {parent=faceAttach, loc=<-6.3271 16.4705 4.0369>, rot=<0 -0.0074 1 0>, scale=<5.9219 5.9219 5.9219>}
 R_lid_lwr_mid {parent=R_lid_lwr_main, loc=<0 0 -1>, rot=<0 0 0 1>, scale=<0.05 0.05 1>}
 R_lid_lwr_1 {parent=R_lid_lwr_main, loc=<0 0 -1>, rot=<0 0 0 -1>, scale=<0.05 0.05 1>}
 R_lid_lwr_2 {parent=R_lid_lwr_main, loc=<0 0 -1>, rot=<0 0 0 1>, scale=<0.05 0.05 1>}
 R_lid_lwr_3 {parent=R_lid_lwr_main, loc=<0 0 -1>, rot=<0 0 0 1>, scale=<0.05 0.05 1>}
 R_lid_lwr_4 {parent=R_lid_lwr_main, loc=<0 0 -1>, rot=<0 0 0 1>, scale=<0.05 0.05 1>}
 R_lid_lwr_5 {parent=R_lid_lwr_main, loc=<0 0 -1>, rot=<0 0 0 1>, scale=<0.05 0.05 1>}
 R_lid_lwr_6 {parent=R_lid_lwr_main, loc=<0 0 -1>, rot=<0 0 0 1>, scale=<0.05 0.05 1>}
 R_lid_lwr_end {parent=R_lid_lwr_main, loc=<0 0 -1.0001>, rot=<0 0 0 1>}
 R_brow_hide {parent=faceAttach, loc=<0 0 0>, rot=<0 0 0 1>}
 L_brow_hide {parent=faceAttach, loc=<0 0 0>, rot=<0 0 0 1>}
 R_browLong_hide {parent=faceAttach, loc=<0 0 0>, rot=<0 0 0 1>}
 L_browLong_hide {parent=faceAttach, loc=<0 0 0>, rot=<0 0 0 1>}
 spine_01_fr_tgt {parent=spine_05, loc=<-0.3101 6.4603 0>, rot=<0 0 -0.024 -0.9997>}
 torso {parent=pelvis, loc=<4.6994 -2.6291 0>, rot=<0 0 0 -1>}
 d_pelvis {parent=torso, loc=<4.0514 7.0022 0>, rot=<0 0 0 -1>}
 d_spine_01 {parent=d_pelvis, loc=<4.7245 -0.4471 0>, rot=<0 0 0.034 -0.9994>}
 d_spine_02 {parent=d_spine_01, loc=<5.0763 -0.0245 0>, rot=<0 0 0.0238 -0.9997>}
 d_spine_03 {parent=d_spine_02, loc=<4.2073 0 0>, rot=<0 0 0.0149 -0.9999>}
 d_spine_04 {parent=d_spine_03, loc=<5.6716 0 0>, rot=<0 0 -0.0132 -0.9999>}
 d_spine_05 {parent=d_spine_04, loc=<7.0102 0 0>, rot=<0 0 -0.0015 -1>}
 d_chest_end {parent=d_spine_05, loc=<10.3678 -0.1438 0>, rot=<0 0 -0.036 -0.9994>}
 d_head_snap {parent=d_chest_end, loc=<13.3974 -2.5756 0>, rot=<0 0 0 -1>}
]
