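# Post-copy AnimBlueprint for the face accessory part of the Ruckus Mini player body, following the parent mesh's pose and curves.
# The AnimGraph adds a spine yaw while the dumbbell emote montage plays, rescales brow-hide bones, then runs FaceAcc and Eyelids Control Rigs, feeding eye twist to the latter.
# The EventGraph caches the owner's AnimInstance to poll that montage and a frontend idle curve, whose flag the AnimGraph does not read.
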AnimBlueprint /BRCosmetics/Characters/Player/Male/Medium/Bodies/M_MED_Ruckus_Mini/Meshes/Parts/M_MED_Ruckus_Mini_FaceAcc_AnimBP
target_skeleton=&/M_MED_Ruckus_Mini_FaceAcc_Skeleton
graphs [
 AnimGraph {nodes [
   n1 = "Transform (Modify) Bone - Bone: d_spine_02"(
    ComponentPose=Local To Component(LocalPose=Copy Pose From Mesh() {bCopyCurves=true, bUseAttachedParent=true})
    Alpha=Get dumbbell() -> (dumbbell:double)
    Translation="0,0,0"
    Rotation="0,-10,0"
    Scale="1,1,1"
   ) {
    rotationMode=BMM_Additive
    rotationSpace=BCS_ParentBoneSpace
    alphaBoolBlend {blendOption=Linear}
    alphaScaleBiasClamp {bInterpResult=true, interpSpeedIncreasing=5}
    boneToModify {boneName=d_spine_02}
    rotation {yaw=-10}
   }
   n2 = "Transform (Modify) Bone - Bone: R_browLong_hide"(ComponentPose=@n1, Alpha=1, Translation="0,0,0", Rotation="0,0,0", Scale="0.1,0.1,0.1") {
    scaleMode=BMM_Replace
    scaleSpace=BCS_ParentBoneSpace
    alphaBoolBlend {blendOption=Linear}
    boneToModify {boneName=R_browLong_hide}
    scale=<0.1 0.1 0.1>
   }
   n3 = "Transform (Modify) Bone - Bone: L_browLong_hide"(ComponentPose=@n2, Alpha=1, Translation="0,0,0", Rotation="0,0,0", Scale="0.1,0.1,0.1") {
    scaleMode=BMM_Replace
    scaleSpace=BCS_ParentBoneSpace
    alphaBoolBlend {blendOption=Linear}
    boneToModify {boneName=L_browLong_hide}
    scale=<0.1 0.1 0.1>
   }
   n4 = "Transform (Modify) Bone - Bone: R_brow_hide"(ComponentPose=@n3, Alpha=1, Translation="0,0,0", Rotation="0,0,0", Scale="1,1,1") {scaleMode=BMM_Replace, scaleSpace=BCS_ParentBoneSpace, alphaBoolBlend {blendOption=Linear}, boneToModify {boneName=R_brow_hide}}
   n5 = "Transform (Modify) Bone - Bone: L_brow_hide"(ComponentPose=@n4, Alpha=1, Translation="0,0,0", Rotation="0,0,0", Scale="1,1,1") {scaleMode=BMM_Replace, scaleSpace=BCS_ParentBoneSpace, alphaBoolBlend {blendOption=Linear}, boneToModify {boneName=L_brow_hide}}
   n6 = "Transform (Modify) Bone - Bone: R_browLong_hide"!(ComponentPose=@n5, bAlphaBoolEnabled:bool=True, Translation="0,0,0", Rotation="0,0,0", Scale="1,1,1") -> (Pose) {
    alphaInputType=Bool
    scaleMode=BMM_Replace
    scaleSpace=BCS_ParentBoneSpace
    alphaBoolBlend {blendOption=Linear}
    boneToModify {boneName=R_browLong_hide}
   }
   n7 = "Transform (Modify) Bone - Bone: L_browLong_hide"!(ComponentPose=@n6, bAlphaBoolEnabled:bool=True, Translation="0,0,0", Rotation="0,0,0", Scale="1,1,1") -> (Pose) {
    alphaInputType=Bool
    scaleMode=BMM_Replace
    scaleSpace=BCS_ParentBoneSpace
    alphaBoolBlend {blendOption=Linear}
    boneToModify {boneName=L_browLong_hide}
   }
   n8 = "Transform (Modify) Bone - Bone: R_brow_hide"!(ComponentPose=@n7, bAlphaBoolEnabled:bool=True, Translation="0,0,0", Rotation="0,0,0", Scale="0.1,0.1,0.1") -> (Pose) {
    alphaInputType=Bool
    scaleMode=BMM_Replace
    scaleSpace=BCS_ParentBoneSpace
    alphaBoolBlend {blendOption=Linear}
    boneToModify {boneName=R_brow_hide}
    scale=<0.1 0.1 0.1>
   }
   n9 = "Transform (Modify) Bone - Bone: L_brow_hide"!(ComponentPose=@n8, bAlphaBoolEnabled:bool=True, Translation="0,0,0", Rotation="0,0,0", Scale="0.1,0.1,0.1") -> (Pose) {
    alphaInputType=Bool
    scaleMode=BMM_Replace
    scaleSpace=BCS_ParentBoneSpace
    alphaBoolBlend {blendOption=Linear}
    boneToModify {boneName=L_brow_hide}
    scale=<0.1 0.1 0.1>
   }
   n10 = Control Rig(Source=Component To Local(ComponentPose=@n9)) {
    alphaBoolBlend {blendOption=Linear}
    controlRigAssetReference {blueprintRigClass=&/M_MED_Ruckus_Mini_FaceAcc_CtrlRig.M_MED_Ruckus_Mini_FaceAcc_CtrlRig_C}
   }
   n11 = Control Rig!(Source=@n10, Alpha, eye_twist:double=Get do_eye_twist() -> (do_eye_twist:double)) -> (Pose) {
    alphaBoolBlend {blendOption=Linear}
    controlRigAssetReference {blueprintRigClass=&/M_MED_Ruckus_Mini_Eyelids_CtrlRig.M_MED_Ruckus_Mini_Eyelids_CtrlRig_C}
   }
   Output Pose(Result=@n11) {layerGroup=None}
  ]}
 EventGraph {nodes [
   Event BlueprintUpdateAnimation()
   > n1 = Sequence()
   > Get!(execute:Exec) -> (then:Exec, else:Exec, Owner_Skel_Mesh:Skeletal Mesh Component Object Reference)
   > Get!(execute) -> (then, else, Anim_Inst:Anim Instance Object Reference)
   n2 = Get Anim_Inst() -> (Anim_Inst:Anim Instance Object Reference)
   > Set PlayingFrontendIdle(PlayingFrontendIdle:bool="float > float"(A=GetCurveValue(self=@n2, CurveName=FrontEnd_Idle_Bypass), B=0) {is=title K2Node_CallFunction}) -> (Output_Get:bool)
   n3 = Get Anim_Inst()
   Set dumbbell(
    execute=@n1.then_1
    dumbbell:double="To Float (Boolean)"(InBool=Montage_IsPlaying(self=@n3, Montage=/BRCosmetics/Animation/Game/MainPlayer/Emotes/Dumbbell/Emote_Dumbbell_Lift_CMM_M))
   ) -> (Output_Get:double)
   Event BlueprintInitializeAnimation()
   > n4 = Sequence()
   Set skelmesh(execute=@n4.then_1, skelmesh:Skeletal Mesh Component Object Reference=GetPartSkeletalMeshComponent()) -> (Output_Get:Skeletal Mesh Component Object Reference)
   > n5 = Set Owner_Skel_Mesh(Owner_Skel_Mesh:Skeletal Mesh Component Object Reference=GetOwnerSkeletalMeshComponent()) -> (Output_Get:Skeletal Mesh Component Object Reference)
   > Set Anim_Inst(Anim_Inst:Anim Instance Object Reference=GetAnimInstance(self=@n5)) -> (Output_Get:Anim Instance Object Reference)
  ]}
]
variables [
 Anim_Inst {type=AnimInstance, default=null}
 do_eye_twist {type=number, default=1}
 dumbbell {type=number}
 Owner_Skel_Mesh {type=SkeletalMeshComponent}
 Pawn {type=FortPlayerPawn}
 PlayingFrontendIdle {type=boolean}
 skelmesh {type=SkeletalMeshComponent}
]
